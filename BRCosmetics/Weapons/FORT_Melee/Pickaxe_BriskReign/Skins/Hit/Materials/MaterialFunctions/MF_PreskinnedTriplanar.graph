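# Material function for triplanar texture projection in pre-skinned local space, keeping the mapping fixed to skinned meshes.
# Position is normalized by the largest pre-skinned bounds extent, then projected into sign-corrected, tiled, offset UVs per axis.
# Contrast-adjusted normal masks blend samples into XYZ or cheaper XY outputs, alongside UVs, blend masks and an axis mask.
# Per-axis parallax offsets derive from the local view vector, with adjustable reduction at grazing angles.
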MaterialFunction /BRCosmetics/Weapons/FORT_Melee/Pickaxe_BriskReign/Skins/Hit/Materials/MaterialFunctions/MF_PreskinnedTriplanar
graph {nodes [
  FunctionInput() {bUsePreviewValueAsDefault=true, inputName=//----------- PARALLAX -----------//, inputType=FunctionInput_Scalar, sortPriority=100}
  FunctionOutput(None=NamedRerouteUsage()) {description="Sampled Output on all 3 axes ", outputName=Sample XYZ}
  FunctionOutput(None=NamedRerouteUsage()) {description=Projected UV for X Axis, outputName=UV X, sortPriority=10}
  FunctionOutput(None=NamedRerouteUsage()) {description=Projected UV for Y Axis, outputName=UV Y, sortPriority=11}
  FunctionOutput(None=NamedRerouteUsage()) {description=Projected UV for Z Axis, outputName=UV Z, sortPriority=12}
  FunctionOutput(None=NamedRerouteUsage()) {description=Blend mask for UV X, outputName=UV Blend X, sortPriority=20}
  FunctionOutput(None=NamedRerouteUsage()) {description=Blend mask for UV Z, outputName=UV Blend Z, sortPriority=21}
  FunctionOutput(None=NamedRerouteUsage()) {description="Sampled output on 2 axes.  Cheaper than XYZ", outputName=Sample XY, sortPriority=1}
  n1 = Constant3Vector() {constant=<rgba -2 2 2 0>}
  n2 = NamedRerouteUsage()
  n3 = Add(A=NamedRerouteDeclaration(None=Step(X=@n2)) {name="NormalLS Positive Mask (V3)", nodeColor=<rgba 0.5234 0 1 1>}) {constA=-0.5, constB=-0.5}
  FunctionOutput(None=NamedRerouteDeclaration(None=Multiply(A=@n1, B=@n3)) {name=UVSwizzle, nodeColor=<rgba 0 1 0.0625 1>}) {outputName=UVSwizzle}
  n4 = LinearInterpolate(
   A=Constant3Vector() {constant=<rgba 0 1 0.5 0>}
   B=Constant3Vector() {constant=<rgba 0 1 0 0>}
   Alpha=ComponentMask(None=NamedRerouteUsage()) {g=true}
  ) {desc=Float3}
  n5 = LinearInterpolate(
   A=Constant3Vector() {constant=<rgba 1 0 0.5 0>}
   B=Constant3Vector() {constant=<rgba 1 0 0 0>}
   Alpha=ComponentMask(None=NamedRerouteUsage()) {r=true}
  ) {desc=Float3}
  n6 = NamedRerouteUsage()
  n7 = Constant3Vector() {constant=<rgba 0 0.5 1 0>}
  n8 = Constant3Vector() {constant=<rgba 0 0 1 0>}
  n9 = ComponentMask(None=NamedRerouteUsage()) {b=true}
  FunctionOutput(None=LinearInterpolate(
    A=LinearInterpolate(A=@n4, B=@n5, Alpha=@n6) {desc=Float3}
    B=LinearInterpolate(A=@n7, B=@n8, Alpha=@n9) {desc=Float3}
    Alpha=NamedRerouteUsage()
   )) {outputName=Axis Mask, sortPriority=100}
  n10 = FunctionInput() {
   bUsePreviewValueAsDefault=true
   description=Blend factor for the projected texture samples.
   inputName=Blend Contrast
   inputType=FunctionInput_Scalar
   sortPriority=4
   previewValue=<1 0 0 1>
  }
  NamedRerouteDeclaration(None=@n10) {name=BlendContrast, nodeColor=<rgba 1 0 0.5625 1>}
  NamedRerouteDeclaration(None=TextureSample(UVs=NamedRerouteUsage(), Tex=NamedRerouteUsage()) {samplerSource=SSM_Wrap_WorldGroupSettings}.RGBA) {name=SampleX, nodeColor=<rgba 1 0.0042 0 1>}
  NamedRerouteDeclaration(None=TextureSample(UVs=NamedRerouteUsage(), Tex=NamedRerouteUsage()) {samplerSource=SSM_Wrap_WorldGroupSettings}.RGBA) {name=SampleY, nodeColor=<rgba 0.0839 1 0 1>}
  NamedRerouteDeclaration(None=TextureSample(UVs=NamedRerouteUsage(), Tex=NamedRerouteUsage()) {samplerSource=SSM_Wrap_WorldGroupSettings}.RGBA) {name=SampleZ, nodeColor=<rgba 0 0.0094 1 1>}
  n11 = NamedRerouteDeclaration(None=VertexInterpolator(VS=PreSkinnedNormal()).PS) {name=NormalLS, nodeColor=<rgba 1 0.7031 0 1>}
  n12 = Abs(None=ComponentMask(None=@n11) {r=true})
  n13 = NamedRerouteUsage()
  NamedRerouteDeclaration(None=ComponentMask(None=MaterialFunctionCall(In=@n12, Contrast=@n13) {materialFunction=&CheapContrast}.Result) {r=true}) {name=BlendX, nodeColor=<rgba 1 0 0.7266 1>}
  n14 = Abs(None=ComponentMask(None=@n11) {b=true})
  NamedRerouteDeclaration(None=ComponentMask(None=MaterialFunctionCall(In=@n14, Contrast=@n13) {materialFunction=&CheapContrast}.Result) {r=true}) {name=BlendZ, nodeColor=<rgba 0 1 0.9531 1>}
  n15 = LinearInterpolate(A=NamedRerouteUsage(), B=NamedRerouteUsage(), Alpha=NamedRerouteUsage()) {desc=Float3}
  NamedRerouteDeclaration(None=LinearInterpolate(A=NamedRerouteDeclaration(None=@n15) {name=SampleXY, nodeColor=<rgba 1 0.9141 0 1>}, B=NamedRerouteUsage(), Alpha=NamedRerouteUsage())) {name=SampleXYZ, nodeColor=<rgba 1 0.6328 0 1>}
  n16 = FunctionInput() {
   bUsePreviewValueAsDefault=true
   description="UV Offset, can be used to warp UVs as well."
   inputName=UV Offset
   inputType=FunctionInput_Vector2
   sortPriority=2
  }
  NamedRerouteDeclaration(None=@n16) {name=UVOffset, nodeColor=<rgba 0.0078 1 0 1>}
  n17 = FunctionInput() {
   bUsePreviewValueAsDefault=true
   description=Scale param for UVs
   inputName=UV Tiling
   inputType=FunctionInput_Vector2
   sortPriority=1
   previewValue=<1 1 0 1>
  }
  NamedRerouteDeclaration(None=@n17) {name="UVTiling (V2)", nodeColor=<rgba 1 0.4688 0 1>}
  n18 = CameraVectorWS()
  NamedRerouteDeclaration(None=FunctionInput(Preview=Transform(None=@n18) {transformSourceType=TRANSFORMSOURCE_World, transformType=TRANSFORM_Local}) {bUsePreviewValueAsDefault=true, inputName=View Vector LS, sortPriority=110}) {name="View Vector LS (V3)", nodeColor=<rgba 0 0.8359 1 1>}
  NamedRerouteDeclaration(None=FunctionInput() {bUsePreviewValueAsDefault=true, inputName=Parallax Amount, inputType=FunctionInput_Scalar, sortPriority=120}) {name="Parallax Intensity (S)", nodeColor=<rgba 1 0.8438 0 1>}
  n19 = NamedRerouteUsage()
  MaterialExpressionMultiply_25 = Multiply(
   A=NamedRerouteUsage()
   B=LinearInterpolate(
    A=Convert!() {convertOutputs [{type=Vector3, defaultValue=<rgba -1 -1 1 1>}]}
    B=Convert() {convertOutputs [{type=Vector3, defaultValue=<rgba 1 1 1 1>}]}
    Alpha=ComponentMask(None=@n19) {r=true}
   )
  )
  n20 = Convert!(None=@MaterialExpressionMultiply_25) {
   convertInputs [{
     type=Vector3
     expressionInput {
      expression=&self:MaterialExpressionMultiply_25
      inputName=None
      mask=0
      maskA=0
      maskB=0
      maskG=0
      maskR=0
      outputIndex=0
     }
    }]
   convertOutputs [{type=Vector2}, {type=Scalar}]
  }
  n21 = Max(A=@n20, B=NamedRerouteUsage()) {constB=0.001}
  n22 = LinearInterpolate(A=NamedRerouteUsage(), Alpha=SmoothStep(Max=NamedRerouteUsage(), Value=Abs(None=@n20)))
  n23 = NamedRerouteUsage()
  NamedRerouteDeclaration(None=Multiply(A=Divide(A=@n20, B=@n21), B=Multiply(A=@n22, B=@n23))) {name="UVX Parallax Offset (V2)", nodeColor=<rgba 1 0 0 1>}
  n24 = NamedRerouteUsage()
  MaterialExpressionMultiply_27 = Multiply(
   A=NamedRerouteUsage()
   B=LinearInterpolate(
    A=Convert!() {convertOutputs [{type=Vector3, defaultValue=<rgba 1 -1 1 1>}]}
    B=Convert() {convertOutputs [{type=Vector3, defaultValue=<rgba -1 1 1 1>}]}
    Alpha=ComponentMask(None=@n24) {g=true}
   )
  )
  n25 = Convert!(None=@MaterialExpressionMultiply_27) {
   convertInputs [{
     type=Vector3
     expressionInput {
      expression=&self:MaterialExpressionMultiply_27
      inputName=None
      mask=0
      maskA=0
      maskB=0
      maskG=0
      maskR=0
      outputIndex=0
     }
    }]
   convertOutputs [{type=Vector2}, {type=Scalar}]
  }
  n26 = Max(A=@n25, B=NamedRerouteUsage()) {constB=0.001}
  n27 = LinearInterpolate(A=NamedRerouteUsage(), Alpha=SmoothStep(Max=NamedRerouteUsage(), Value=Abs(None=@n25)))
  n28 = NamedRerouteUsage()
  NamedRerouteDeclaration(None=Multiply(A=Divide(A=@n25, B=@n26), B=Multiply(A=@n27, B=@n28))) {name="UVY Parallax Offset (V2)", nodeColor=<rgba 0 1 0 1>}
  n29 = NamedRerouteUsage()
  MaterialExpressionMultiply_29 = Multiply(
   A=NamedRerouteUsage()
   B=LinearInterpolate(
    A=Convert!() {convertOutputs [{type=Vector3, defaultValue=<rgba -1 1 -1 1>}]}
    B=Convert() {convertOutputs [{type=Vector3, defaultValue=<rgba -1 -1 1 1>}]}
    Alpha=ComponentMask(None=@n29) {b=true}
   )
  )
  n30 = Convert!(None=@MaterialExpressionMultiply_29) {
   convertInputs [{
     type=Vector3
     expressionInput {
      expression=&self:MaterialExpressionMultiply_29
      inputName=None
      mask=0
      maskA=0
      maskB=0
      maskG=0
      maskR=0
      outputIndex=0
     }
    }]
   convertOutputs [{type=Vector2}, {type=Scalar}]
  }
  n31 = Max(A=@n30, B=NamedRerouteUsage()) {constB=0.001}
  n32 = LinearInterpolate(A=NamedRerouteUsage(), Alpha=SmoothStep(Max=NamedRerouteUsage(), Value=Abs(None=@n30)))
  n33 = NamedRerouteUsage()
  NamedRerouteDeclaration(None=Multiply(A=Divide(A=@n30, B=@n31), B=Multiply(A=@n32, B=@n33))) {name="UVZ Parallax Offset (V2)", nodeColor=<rgba 0 0 1 1>}
  n34 = FunctionInput() {bUsePreviewValueAsDefault=true, inputName=Projection Min Cos, inputType=FunctionInput_Scalar, sortPriority=130, previewValue=<0.1 0 0 1>}
  NamedRerouteDeclaration(None=@n34) {name="Projection Min Cos (S)", nodeColor=<rgba 0 1 0.9531 1>}
  n35 = FunctionInput() {
   bUsePreviewValueAsDefault=true
   inputName=Grazing Angle Flatten Min Cos
   inputType=FunctionInput_Scalar
   sortPriority=150
   previewValue=<0.1 0 0 1>
  }
  NamedRerouteDeclaration(None=@n35) {name="Reduction Min Cos (S)", nodeColor=<rgba 1 0.6328 0 1>}
  n36 = FunctionInput() {
   bUsePreviewValueAsDefault=true
   inputName=Grazing Angle Parallax Mult
   inputType=FunctionInput_Scalar
   sortPriority=140
   previewValue=<0.1 0 0 1>
  }
  NamedRerouteDeclaration(None=@n36) {name="Parallax Reduction (S)", nodeColor=<rgba 0.3125 0 1 1>}
  MaterialExpressionPreSkinnedLocalBounds_1 = PreSkinnedLocalBounds()
  n37 = Subtract(A=VertexInterpolator(VS=LocalPosition() {localOrigin=InstancePreSkinning}.XYZ).PS, B=@MaterialExpressionPreSkinnedLocalBounds_1.Min)
  n38 = Convert(None=@MaterialExpressionPreSkinnedLocalBounds_1.Extents) {
   nodeName=Break Float3
   convertInputs [{
     type=Vector3
     expressionInput {
      expression=&self:MaterialExpressionPreSkinnedLocalBounds_1
      inputName=None
      mask=1
      maskA=0
      maskB=1
      maskG=1
      maskR=1
      outputIndex=1
     }
    }]
   convertOutputs [{type=Scalar}, {type=Scalar}, {type=Scalar}]
  }
  n39 = Max(A=Max(A=@n38, B=@n38), B=@n38)
  n40 = Multiply(
   A=Subtract(A=Divide(A=@n39, B=@MaterialExpressionPreSkinnedLocalBounds_1.Extents))
   B=@MaterialExpressionPreSkinnedLocalBounds_1.Half Extents
  )
  NamedRerouteDeclaration(None=Divide(A=Add(A=@n37, B=@n40), B=@n39)) {name=UV0-1, nodeColor=<rgba 0.6172 0 1 1>}
  n41 = FunctionInput(Preview=TextureObject() {texture=&T_BlacksRGB}) {
   bUsePreviewValueAsDefault=true
   description=Texture to be sampled with triplanar uvs
   inputName=InputTexture
   inputType=FunctionInput_Texture2D
  }
  NamedRerouteDeclaration(None=@n41) {name=TextureDefault, nodeColor=<rgba 0 0.4844 1 1>}
  n42 = NamedRerouteUsage()
  n43 = ComponentMask(None=NamedRerouteUsage()) {r=true}
  n44 = Constant() {r=-1}
  n45 = Add(A=Multiply(A=ComponentMask(None=@n42) {b=true, g=true}, B=AppendVector(A=@n43, B=@n44)), B=NamedRerouteUsage())
  n46 = NamedRerouteUsage()
  NamedRerouteDeclaration(None=Add(A=Multiply(A=@n45, B=@n46), B=NamedRerouteUsage())) {name=UVX, nodeColor=<rgba 1 0 0.0227 1>}
  n47 = NamedRerouteUsage()
  n48 = ComponentMask(None=NamedRerouteUsage()) {g=true}
  n49 = Constant() {r=-1}
  n50 = Add(A=Multiply(A=ComponentMask(None=@n47) {b=true, r=true}, B=AppendVector(A=@n48, B=@n49)), B=NamedRerouteUsage())
  n51 = NamedRerouteUsage()
  NamedRerouteDeclaration(None=Add(A=Multiply(A=@n50, B=@n51), B=NamedRerouteUsage())) {name=UVY, nodeColor=<rgba 0.1448 1 0.0938 1>}
  n52 = NamedRerouteUsage()
  n53 = Constant() {r=1}
  n54 = ComponentMask(None=NamedRerouteUsage()) {b=true}
  n55 = Add(A=Multiply(A=ComponentMask(None=@n52) {g=true, r=true}, B=AppendVector(A=@n53, B=@n54)), B=NamedRerouteUsage())
  n56 = NamedRerouteUsage()
  NamedRerouteDeclaration(None=Add(A=Multiply(A=@n55, B=@n56), B=NamedRerouteUsage())) {name=UVZ, nodeColor=<rgba 0 0.0553 1 1>}
 ]}
