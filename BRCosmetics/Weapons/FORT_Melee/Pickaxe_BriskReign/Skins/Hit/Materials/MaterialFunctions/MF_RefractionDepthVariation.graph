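# Material function for the BriskReign pickaxe that outputs a scalar noise mask for refraction depth variation, apparently to fake interior detail.
# A preskinned triplanar noise sample is shaped by power, multiply and add into a variable parallax depth.
# A second triplanar sample takes parallax along the local-space refraction vector, with grazing-angle controls, then is remapped and powered for output.
# Time-driven offsets scroll the two samples in opposite directions at different rates.
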MaterialFunction /BRCosmetics/Weapons/FORT_Melee/Pickaxe_BriskReign/Skins/Hit/Materials/MaterialFunctions/MF_RefractionDepthVariation
graph {nodes [
  n1 = NamedRerouteUsage()
  n2 = NamedRerouteUsage()
  n3 = Constant()
  n4 = Multiply(A=NamedRerouteUsage()) {constB=-0.5625}
  n5 = NamedRerouteUsage()
  n6 = NamedRerouteUsage()
  n7 = ComponentMask(None=MaterialFunctionCall!(
    InputTexture=NamedRerouteUsage()
    UV Tiling=Divide(A=@n1, B=@n2)
    UV Offset=AppendVector(A=@n3, B=@n4)
    Blend Contrast=NamedRerouteUsage()
    //----------- PARALLAX -----------//
    View Vector LS=NamedRerouteUsage()
    Parallax Amount=Divide(A=@n5, B=@n6)
    Projection Min Cos=NamedRerouteUsage()
    Grazing Angle Parallax Mult=NamedRerouteUsage()
    Grazing Angle Flatten Min Cos=NamedRerouteUsage()
   ) {materialFunction=&/MF_PreskinnedTriplanar}.Sample XYZ) {r=true}
  n8 = NamedRerouteUsage()
  n9 = NamedRerouteUsage()
  FunctionOutput(None=Power(
    Base=MaterialFunctionCall!(Input=@n7, Input Low=@n8, Input High=@n9) {materialFunction=/Engine/Functions/Engine_MaterialFunctions03/Math/RemapValueRangeNormalized}.Result
    Exp=NamedRerouteUsage()
   )) {outputName="Result (S)"}
  n10 = FunctionInput() {bUsePreviewValueAsDefault=true, inputName=Triplanar Tiling, inputType=FunctionInput_Scalar, sortPriority=10, previewValue=<1 0 0 1>}
  MaterialExpressionFunctionInput_12 = FunctionInput() {bUsePreviewValueAsDefault=true, inputName=Triplanar Tiling V Ratio, inputType=FunctionInput_Scalar, sortPriority=20, previewValue=<1 0 0 1>}
  n11 = Convert(None=@MaterialExpressionFunctionInput_12) {
   convertInputs [{
     type=Scalar
     expressionInput {
      expression=&self:MaterialExpressionFunctionInput_12
      inputName=None
      mask=0
      maskA=0
      maskB=0
      maskG=0
      maskR=0
      outputIndex=0
     }
    }]
   convertOutputs [{type=Vector2, defaultValue=<rgba 1 0 0 1>}]
  }
  NamedRerouteDeclaration(None=Multiply(A=@n10, B=@n11)) {name="Triplanar Tiling (V2)", nodeColor=<rgba 0 1 0.8828 1>}
  MaterialExpressionFunctionInput_1 = FunctionInput() {
   bUsePreviewValueAsDefault=true
   inputName=Noise Depth Breakup Settings
   inputType=FunctionInput_Vector4
   sortPriority=110
   previewValue=<5 0.1 0.8 2.2>
  }
  n12 = Convert(None=@MaterialExpressionFunctionInput_1) {
   nodeName=Break Float4
   convertInputs [{
     type=Vector4
     defaultValue=<rgba 0 0 0 1>
     expressionInput {
      expression=&self:MaterialExpressionFunctionInput_1
      inputName=None
      mask=0
      maskA=0
      maskB=0
      maskG=0
      maskR=0
      outputIndex=0
     }
    }]
   convertOutputs [{type=Scalar}, {type=Scalar}, {type=Scalar}, {type=Scalar}]
  }
  NamedRerouteDeclaration(None=@n12) {name="Noise Depth Variation Mult (S)", nodeColor=<rgba 0 0.0156 1 1>}
  NamedRerouteDeclaration(None=@n12) {name="Noise Add Constant Depth (S)", nodeColor=<rgba 0.3359 1 0 1>}
  n13 = FunctionInput() {bUsePreviewValueAsDefault=true, inputName=Parallax Depth Norm, inputType=FunctionInput_Scalar, sortPriority=210, previewValue=<1 0 0 1>}
  NamedRerouteDeclaration(None=@n13) {name="Parallax Depth Norm (S)", nodeColor=<rgba 0 0.8594 1 1>}
  n14 = FunctionInput(Preview=TextureObject() {texture=/BRCosmetics/Animation/Props/Gadgets_SatinCheddar/Materials/T_Perlin_Noise_M1}) {bUsePreviewValueAsDefault=true, inputName=Noise Texture, inputType=FunctionInput_Texture2D, previewValue=<1 0 0 1>}
  NamedRerouteDeclaration(None=@n14) {name="Noise Texture (TO)", nodeColor=<rgba 0 0.2031 1 1>}
  MaterialExpressionFunctionInput_0 = FunctionInput() {
   bUsePreviewValueAsDefault=true
   inputName=Variation Noise Settings
   inputType=FunctionInput_Vector4
   sortPriority=100
   previewValue=<0.15 0 0.5 0>
  }
  n15 = Convert(None=@MaterialExpressionFunctionInput_0) {
   nodeName=Break Float4
   convertInputs [{
     type=Vector4
     defaultValue=<rgba 0 0 0 1>
     expressionInput {
      expression=&self:MaterialExpressionFunctionInput_0
      inputName=None
      mask=0
      maskA=0
      maskB=0
      maskG=0
      maskR=0
      outputIndex=0
     }
    }]
   convertOutputs [{type=Scalar}, {type=Scalar}, {type=Scalar}, {type=Scalar}]
  }
  NamedRerouteDeclaration(None=@n15) {name="Interior Noise Scale (S)", nodeColor=<rgba 1 0 0.0938 1>}
  n16 = FunctionInput() {bUsePreviewValueAsDefault=true, inputName=Projection Min Cos, inputType=FunctionInput_Scalar, sortPriority=220, previewValue=<0.1 0 0 1>}
  NamedRerouteDeclaration(None=@n16) {name="Projection Min Cos (S)", nodeColor=<rgba 0 1 0.9531 1>}
  n17 = FunctionInput() {bUsePreviewValueAsDefault=true, inputName=Triplanar Blend Contrast, inputType=FunctionInput_Scalar, sortPriority=30, previewValue=<1 0 0 1>}
  NamedRerouteDeclaration(None=@n17) {name="Triplanar Blend Contrast (S)", nodeColor=<rgba 0 0.8594 1 1>}
  n18 = FunctionInput() {
   bUsePreviewValueAsDefault=true
   inputName=Grazing Angle Flatten Min Cos
   inputType=FunctionInput_Scalar
   sortPriority=240
   previewValue=<0.1 0 0 1>
  }
  NamedRerouteDeclaration(None=@n18) {name="Reduction Min Cos (S)", nodeColor=<rgba 1 0.6328 0 1>}
  n19 = FunctionInput() {
   bUsePreviewValueAsDefault=true
   inputName=Grazing Angle Parallax Mult
   inputType=FunctionInput_Scalar
   sortPriority=230
   previewValue=<0.1 0 0 1>
  }
  NamedRerouteDeclaration(None=@n19) {name="Parallax Reduction (S)", nodeColor=<rgba 0.3125 0 1 1>}
  n20 = FunctionInput(Preview=Multiply(A=CameraVectorWS()) {constB=-1}) {bUsePreviewValueAsDefault=true, inputName=Refraction Vector WS, sortPriority=200}
  NamedRerouteDeclaration(None=Transform(None=Multiply(A=@n20) {constB=-1}) {transformSourceType=TRANSFORMSOURCE_World, transformType=TRANSFORM_Local}) {name="View Vector LS (V3)", nodeColor=<rgba 1 0.6797 0 1>}
  NamedRerouteDeclaration(None=@n15) {name="Interior Noise OldMin (S)", nodeColor=<rgba 0 0.4141 1 1>}
  NamedRerouteDeclaration(None=@n15) {name="Interior Noise OldMax (S)", nodeColor=<rgba 0.7344 1 0 1>}
  n21 = FunctionInput(Preview=Multiply(A=Time()) {constB=0.004}) {bUsePreviewValueAsDefault=true, inputName=Time, inputType=FunctionInput_Scalar, sortPriority=120, previewValue=<1 0 0 1>}
  NamedRerouteDeclaration(None=@n21) {name="Time (S)", nodeColor=<rgba 0 0.7656 1 1>}
  n22 = NamedRerouteUsage()
  n23 = Divide(A=NamedRerouteUsage(), B=NamedRerouteUsage())
  n24 = AppendVector(A=Constant(), B=Multiply(A=NamedRerouteUsage()) {constB=0.5})
  n25 = NamedRerouteUsage()
  n26 = Power(
   Base=ComponentMask(None=MaterialFunctionCall(InputTexture=@n22, UV Tiling=@n23, UV Offset=@n24, Blend Contrast=@n25) {materialFunction=&MF_PreskinnedTriplanar}.Sample XYZ) {r=true}
   Exp=NamedRerouteUsage()
  )
  n27 = NamedRerouteUsage()
  NamedRerouteDeclaration(None=Add(A=Multiply(A=@n26, B=@n27), B=NamedRerouteUsage())) {name="Noise Parallax Variation (S)", nodeColor=<rgba 1 0 0.9844 1>}
  NamedRerouteDeclaration(None=@n15) {name="Interior Noise Exp (S)", nodeColor=<rgba 0.9453 0 1 1>}
  NamedRerouteDeclaration(None=@n12) {name="Noise Depth Variation Scale (S)", nodeColor=<rgba 0 1 0.625 1>}
  NamedRerouteDeclaration(None=@n12) {name="Noise Depth Variation Exp (S)", nodeColor=<rgba 1 0.3047 0 1>}
 ]}
