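# Material function for the BriskReign pickaxe hit skin that layers volumetric ice onto an incoming material, taking tech-art mask and interior emission inputs.
# It refracts the view vector to sample environment illuminance and offset UVs, then feeds thickness, absorption and scattering into a dielectric interior volume for fake transparency.
# Rod and ring region masks set detail UV scales, while island-UV wrapping and refraction depth noise add interior parallax detail.
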MaterialFunction /BRCosmetics/Weapons/FORT_Melee/Pickaxe_BriskReign/Skins/Hit/Materials/MaterialFunctions/MF_VolumetricIce_Pickaxe_BriskReign_Hit
graph {expressions {
  MaterialExpressionAdd_28 {
   class=MaterialExpressionAdd
   inputs [
    A {from=MaterialExpressionMultiply_79}
    B {from=MaterialExpressionNamedRerouteUsage_381}
   ]
  }
  MaterialExpressionAdd_29 {
   class=MaterialExpressionAdd
   inputs [
    A {from=MaterialExpressionNamedRerouteUsage_483}
    B {from=MaterialExpressionMultiply_78}
   ]
  }
  MaterialExpressionAdd_34 {
   class=MaterialExpressionAdd
   inputs [
    A {from=MaterialExpressionMultiply_84}
    B {from=MaterialExpressionNamedRerouteUsage_391}
   ]
  }
  MaterialExpressionAdd_35 {
   class=MaterialExpressionAdd
   inputs [
    A {from=MaterialExpressionConvert_22}
    B {from=""}
   ]
   properties {constB=0.5}
  }
  MaterialExpressionAdd_36 {
   class=MaterialExpressionAdd
   inputs [
    A {from=MaterialExpressionNamedRerouteUsage_432}
    B {from=MaterialExpressionAdd_35}
   ]
  }
  MaterialExpressionAdd_37 {
   class=MaterialExpressionAdd
   inputs [
    A {from=MaterialExpressionNamedRerouteUsage_443}
    B {from=MaterialExpressionNamedRerouteUsage_489}
   ]
  }
  MaterialExpressionAdd_38 {
   class=MaterialExpressionAdd
   inputs [
    A {from=MaterialExpressionAdd_29}
    B {from=MaterialExpressionNamedRerouteUsage_492}
   ]
  }
  MaterialExpressionAdd_41 {
   class=MaterialExpressionAdd
   inputs [
    A {from=MaterialExpressionMultiply_113}
    B {from=MaterialExpressionNamedRerouteUsage_549}
   ]
   properties {desc=Emission Tiling Island UV}
  }
  MaterialExpressionAdd_42 {
   class=MaterialExpressionAdd
   inputs [
    A {from=MaterialExpressionMultiply_114}
    B {from=MaterialExpressionNamedRerouteUsage_559}
   ]
   properties {desc=Diffuse Tiling Island UV}
  }
  MaterialExpressionAppendVector_17 {
   class=MaterialExpressionAppendVector
   inputs [
    A {from=MaterialExpressionVectorParameter_65, output=R}
    B {from=MaterialExpressionVectorParameter_65, output=R}
   ]
  }
  MaterialExpressionAppendVector_18 {
   class=MaterialExpressionAppendVector
   inputs [
    A {from=MaterialExpressionVectorParameter_65, output=B}
    B {from=MaterialExpressionVectorParameter_65, output=B}
   ]
  }
  MaterialExpressionBlendMaterialAttributes_2 {
   class=MaterialExpressionBlendMaterialAttributes
   inputs [
    A {from=MaterialExpressionNamedRerouteUsage_387, output=""}
    B {from=MaterialExpressionNamedRerouteUsage_386, output=""}
    Alpha {from=MaterialExpressionNamedRerouteUsage_457, output=""}
   ]
   properties {}
  }
  MaterialExpressionCameraVectorWS_1 {class=MaterialExpressionCameraVectorWS, inputs [], properties {}}
  MaterialExpressionChannelMaskParameter_1 {
   class=MaterialExpressionChannelMaskParameter
   inputs [None {from=MaterialExpressionNamedRerouteUsage_6, output=""}]
   properties {group=Ice | 01_Globals, maskChannel=Blue, parameterName=Rod Mask Channel, sortPriority=310, defaultValue=<rgba 0 0 1 0>}
  }
  MaterialExpressionChannelMaskParameter_2 {
   class=MaterialExpressionChannelMaskParameter
   inputs [None {from=MaterialExpressionNamedRerouteUsage_7, output=""}]
   properties {group=Ice | 01_Globals, maskChannel=Green, parameterName=Top Ring Mask Channel, sortPriority=320, defaultValue=<rgba 0 1 0 0>}
  }
  MaterialExpressionChannelMaskParameter_3 {
   class=MaterialExpressionChannelMaskParameter
   inputs [None {from=MaterialExpressionNamedRerouteUsage_8, output=""}]
   properties {group=Ice | 01_Globals, maskChannel=Blue, parameterName=Middle Ring Mask Channel, sortPriority=330, defaultValue=<rgba 0 0 1 0>}
  }
  MaterialExpressionChannelMaskParameter_8 {
   class=MaterialExpressionChannelMaskParameter
   inputs [None {from=MaterialExpressionNamedRerouteUsage_9, output=""}]
   properties {group=Ice | 01_Globals, maskChannel=Green, parameterName=Bottom Ring Mask Channel, sortPriority=340, defaultValue=<rgba 0 1 0 0>}
  }
  MaterialExpressionClamp_2 {
   class=MaterialExpressionClamp
   inputs [
    None {from=MaterialExpressionScalarParameter_67, output=""}
    Min {from="", output=""}
    Max {from="", output=""}
   ]
   properties {maxDefault=5, minDefault=1}
  }
  MaterialExpressionComponentMask_6 {class=MaterialExpressionComponentMask, inputs [None {from=MaterialExpressionNamedRerouteUsage_378}], properties {g=true, r=true}}
  MaterialExpressionComponentMask_7 {class=MaterialExpressionComponentMask, inputs [None {from=MaterialExpressionNamedRerouteUsage_378}], properties {b=true}}
  MaterialExpressionComponentMask_8 {
   class=MaterialExpressionComponentMask
   inputs [None {from=MaterialExpressionNamedRerouteUsage_407}]
   properties {b=true, desc="dot(VertexN, RefrDir)"}
  }
  MaterialExpressionConstant_30 {class=MaterialExpressionConstant, properties {r=0.35}}
  MaterialExpressionConstant_31 {class=MaterialExpressionConstant, properties {r=1.25}}
  MaterialExpressionConstant_32 {class=MaterialExpressionConstant, properties {r=1}}
  MaterialExpressionConstant_33 {class=MaterialExpressionConstant}
  MaterialExpressionConstant_34 {class=MaterialExpressionConstant, properties {r=10000}}
  MaterialExpressionConstant_35 {class=MaterialExpressionConstant}
  MaterialExpressionConstant_38 {class=MaterialExpressionConstant, properties {r=1.2}}
  MaterialExpressionConstant_39 {class=MaterialExpressionConstant, properties {r=0.55}}
  MaterialExpressionConstant_40 {class=MaterialExpressionConstant, properties {r=0.2}}
  MaterialExpressionConstant_41 {class=MaterialExpressionConstant}
  MaterialExpressionConvert_1 {
   class=MaterialExpressionConvert
   inputs [None {from=MaterialExpressionNamedRerouteUsage_27}]
   properties {
    nodeName=Break Float4
    convertInputs [{
      type=Vector4
      defaultValue=<rgba 0 0 0 1>
      expressionInput {
       expression=&self:MaterialExpressionNamedRerouteUsage_27
       inputName=None
       mask=0
       maskA=0
       maskB=0
       maskG=0
       maskR=0
       outputIndex=0
      }
     }]
    convertOutputs [{type=Scalar}, {type=Scalar}, {type=Scalar}, {type=Scalar}]
   }
  }
  MaterialExpressionConvert_21 {
   class=MaterialExpressionConvert
   inputs [
    None {from=MaterialExpressionDotProduct_10}
    None {from=MaterialExpressionDotProduct_11}
    None {from=MaterialExpressionDotProduct_12}
   ]
   properties {
    nodeName=Make Vector3
    convertInputs [
     {
      type=Scalar
      expressionInput {
       expression=&self:MaterialExpressionDotProduct_10
       inputName=None
       mask=0
       maskA=0
       maskB=0
       maskG=0
       maskR=0
       outputIndex=0
      }
     }
     {
      type=Scalar
      expressionInput {
       expression=&self:MaterialExpressionDotProduct_11
       inputName=None
       mask=0
       maskA=0
       maskB=0
       maskG=0
       maskR=0
       outputIndex=0
      }
     }
     {
      type=Scalar
      expressionInput {
       expression=&self:MaterialExpressionDotProduct_12
       inputName=None
       mask=0
       maskA=0
       maskB=0
       maskG=0
       maskR=0
       outputIndex=0
      }
     }
    ]
    convertOutputs [{type=Vector3}]
   }
  }
  MaterialExpressionConvert_22 {
   class=MaterialExpressionConvert
   inputs [None {from=MaterialExpressionRgbToHsv_2}]
   properties {
    nodeName=Break Float3
    convertInputs [{
      type=Vector3
      expressionInput {
       expression=&self:MaterialExpressionRgbToHsv_2
       inputName=None
       mask=0
       maskA=0
       maskB=0
       maskG=0
       maskR=0
       outputIndex=0
      }
     }]
    convertOutputs [{type=Scalar}, {type=Scalar}, {type=Scalar}]
   }
  }
  MaterialExpressionConvert_23 {
   class=MaterialExpressionConvert
   inputs [
    None {from=MaterialExpressionFrac_4}
    None {from=MaterialExpressionSaturate_28}
    None {from=MaterialExpressionDivide_34}
   ]
   properties {
    nodeName=Make Vector3
    convertInputs [
     {
      type=Scalar
      expressionInput {
       expression=&self:MaterialExpressionFrac_4
       inputName=None
       mask=0
       maskA=0
       maskB=0
       maskG=0
       maskR=0
       outputIndex=0
      }
     }
     {
      type=Scalar
      expressionInput {
       expression=&self:MaterialExpressionSaturate_28
       inputName=None
       mask=0
       maskA=0
       maskB=0
       maskG=0
       maskR=0
       outputIndex=0
      }
     }
     {
      type=Scalar
      expressionInput {
       expression=&self:MaterialExpressionDivide_34
       inputName=None
       mask=0
       maskA=0
       maskB=0
       maskG=0
       maskR=0
       outputIndex=0
      }
     }
    ]
    convertOutputs [{type=Vector3}]
   }
  }
  MaterialExpressionConvert_31 {
   class=MaterialExpressionConvert
   inputs [None {from=MaterialExpressionMultiply_111}]
   properties {
    nodeName=Break Float2
    convertInputs [{
      type=Vector2
      expressionInput {
       expression=&self:MaterialExpressionMultiply_111
       inputName=None
       mask=0
       maskA=0
       maskB=0
       maskG=0
       maskR=0
       outputIndex=0
      }
     }]
    convertOutputs [{type=Scalar}, {type=Scalar}]
   }
  }
  MaterialExpressionDDX_0 {class=MaterialExpressionDDX, inputs [Value {from=MaterialExpressionNamedRerouteDeclaration_277, output=""}], properties {}}
  MaterialExpressionDDY_0 {class=MaterialExpressionDDY, inputs [Value {from=MaterialExpressionNamedRerouteDeclaration_277, output=""}], properties {}}
  MaterialExpressionDivide_26 {
   class=MaterialExpressionDivide
   inputs [
    A {from=MaterialExpressionNamedRerouteUsage_375}
    B {from=MaterialExpressionMax_14}
   ]
  }
  MaterialExpressionDivide_27 {
   class=MaterialExpressionDivide
   inputs [
    A {from=""}
    B {from=MaterialExpressionNamedRerouteUsage_408}
   ]
  }
  MaterialExpressionDivide_28 {
   class=MaterialExpressionDivide
   inputs [
    A {from=MaterialExpressionComponentMask_6}
    B {from=MaterialExpressionMax_16}
   ]
  }
  MaterialExpressionDivide_29 {
   class=MaterialExpressionDivide
   inputs [
    A {from=MaterialExpressionNamedRerouteUsage_455}
    B {from=MaterialExpressionNamedRerouteUsage_388}
   ]
  }
  MaterialExpressionDivide_33 {
   class=MaterialExpressionDivide
   inputs [
    A {from=MaterialExpressionMultiply_88}
    B {from=MaterialExpressionNamedRerouteUsage_392}
   ]
  }
  MaterialExpressionDivide_34 {
   class=MaterialExpressionDivide
   inputs [
    A {from=""}
    B {from=MaterialExpressionMax_17}
   ]
  }
  MaterialExpressionDivide_35 {
   class=MaterialExpressionDivide
   inputs [
    A {from=MaterialExpressionNamedRerouteUsage_424}
    B {from=MaterialExpressionLinearInterpolate_5}
   ]
  }
  MaterialExpressionDivide_38 {
   class=MaterialExpressionDivide
   inputs [
    A {from=MaterialExpressionSubtract_3}
    B {from=MaterialExpressionNamedRerouteUsage_553}
   ]
   properties {desc=Emission Normalized Island UV}
  }
  MaterialExpressionDivide_39 {
   class=MaterialExpressionDivide
   inputs [
    A {from=MaterialExpressionSubtract_5}
    B {from=MaterialExpressionNamedRerouteUsage_560}
   ]
   properties {desc=Diffuse Normalized Island UV}
  }
  MaterialExpressionDotProduct_10 {
   class=MaterialExpressionDotProduct
   inputs [
    A {from=MaterialExpressionNamedRerouteDeclaration_266, output=""}
    B {from=MaterialExpressionMaterialFunctionCall_23, output=X U}
   ]
   properties {}
  }
  MaterialExpressionDotProduct_11 {
   class=MaterialExpressionDotProduct
   inputs [
    A {from=MaterialExpressionNamedRerouteDeclaration_266, output=""}
    B {from=MaterialExpressionMaterialFunctionCall_23, output=Y V}
   ]
   properties {}
  }
  MaterialExpressionDotProduct_12 {
   class=MaterialExpressionDotProduct
   inputs [
    A {from=MaterialExpressionNamedRerouteDeclaration_266, output=""}
    B {from=MaterialExpressionMaterialFunctionCall_23, output=Z W}
   ]
   properties {}
  }
  MaterialExpressionDotProduct_13 {
   class=MaterialExpressionDotProduct
   inputs [
    A {from=MaterialExpressionNamedRerouteUsage_380, output=""}
    B {from=MaterialExpressionNamedRerouteUsage_422, output=""}
   ]
   properties {}
  }
  MaterialExpressionDotProduct_14 {
   class=MaterialExpressionDotProduct
   inputs [
    A {from=MaterialExpressionVertexNormalWS_2, output=""}
    B {from=MaterialExpressionNamedRerouteDeclaration_296, output=""}
   ]
   properties {}
  }
  MaterialExpressionFrac_4 {class=MaterialExpressionFrac, inputs [None {from=MaterialExpressionAdd_36, output=""}], properties {}}
  MaterialExpressionFrac_5 {class=MaterialExpressionFrac, inputs [None {from=MaterialExpressionDivide_38, output=""}], properties {}}
  MaterialExpressionFrac_6 {class=MaterialExpressionFrac, inputs [None {from=MaterialExpressionDivide_39, output=""}], properties {}}
  MaterialExpressionFunctionInput_0 {
   class=MaterialExpressionFunctionInput
   inputs [Preview {from="", output=""}]
   properties {inputName=Material, inputType=FunctionInput_MaterialAttributes}
  }
  MaterialExpressionFunctionInput_1 {class=MaterialExpressionFunctionInput, inputs [Preview {from="", output=""}], properties {inputName=Tech Art Mask, sortPriority=10}}
  MaterialExpressionFunctionInput_2 {
   class=MaterialExpressionFunctionInput
   inputs [Preview {from="", output=""}]
   properties {bUsePreviewValueAsDefault=true, inputName=Additional Interior Emission, sortPriority=20}
  }
  MaterialExpressionFunctionOutput_1 {
   class=MaterialExpressionFunctionOutput
   inputs [None {from=MaterialExpressionStaticSwitchParameter_2, output=""}]
   properties {outputName="Result (MA)"}
  }
  MaterialExpressionGetMaterialAttributes_0 {
   class=MaterialExpressionGetMaterialAttributes
   inputs [None {from=MaterialExpressionNamedRerouteUsage_568, output=""}]
   properties {attributeGetTypes [
     #1
     #2
     #3
     #4
     #5
    ]}
  }
  MaterialExpressionHsvToRgb_2 {class=MaterialExpressionHsvToRgb, inputs [None {from=MaterialExpressionConvert_23, output=""}], properties {}}
  MaterialExpressionLinearInterpolate_2 {
   class=MaterialExpressionLinearInterpolate
   inputs [
    A {from=MaterialExpressionNamedRerouteUsage_1}
    B {from=MaterialExpressionNamedRerouteUsage_2}
    Alpha {from=MaterialExpressionNamedRerouteUsage_23}
   ]
   properties {constA=0.1458, constB=0.1408}
  }
  MaterialExpressionLinearInterpolate_3 {
   class=MaterialExpressionLinearInterpolate
   inputs [
    A {from=MaterialExpressionLinearInterpolate_2}
    B {from=MaterialExpressionNamedRerouteUsage_3}
    Alpha {from=MaterialExpressionNamedRerouteUsage_21}
   ]
   properties {constB=0.178}
  }
  MaterialExpressionLinearInterpolate_30 {
   class=MaterialExpressionLinearInterpolate
   inputs [
    A {from=MaterialExpressionNamedRerouteUsage_439}
    B {from=MaterialExpressionNamedRerouteUsage_438}
    Alpha {from=MaterialExpressionPower_34}
   ]
  }
  MaterialExpressionLinearInterpolate_31 {
   class=MaterialExpressionLinearInterpolate
   inputs [
    A {from=""}
    B {from=MaterialExpressionMultiply_94}
    Alpha {from=MaterialExpressionNamedRerouteUsage_450}
   ]
   properties {constA=1}
  }
  MaterialExpressionLinearInterpolate_32 {
   class=MaterialExpressionLinearInterpolate
   inputs [
    A {from=MaterialExpressionSaturate_27}
    B {from=MaterialExpressionNamedRerouteUsage_459}
    Alpha {from=""}
   ]
   properties {constAlpha=0.25, desc=to avoid glowing cracks}
  }
  MaterialExpressionLinearInterpolate_36 {
   class=MaterialExpressionLinearInterpolate
   inputs [
    A {from=MaterialExpressionSaturate_37}
    B {from=MaterialExpressionSaturate_30}
    Alpha {from=MaterialExpressionNamedRerouteUsage_495}
   ]
  }
  MaterialExpressionLinearInterpolate_4 {
   class=MaterialExpressionLinearInterpolate
   inputs [
    A {from=MaterialExpressionLinearInterpolate_3}
    B {from=MaterialExpressionNamedRerouteUsage_4}
    Alpha {from=MaterialExpressionNamedRerouteUsage_24}
   ]
   properties {constB=0.177}
  }
  MaterialExpressionLinearInterpolate_44 {
   class=MaterialExpressionLinearInterpolate
   inputs [
    A {from=MaterialExpressionNamedRerouteUsage_551}
    B {from=MaterialExpressionAdd_41}
    Alpha {from=MaterialExpressionNamedRerouteUsage_547}
   ]
  }
  MaterialExpressionLinearInterpolate_45 {
   class=MaterialExpressionLinearInterpolate
   inputs [
    A {from=MaterialExpressionNamedRerouteUsage_561}
    B {from=MaterialExpressionAdd_42}
    Alpha {from=MaterialExpressionNamedRerouteUsage_557}
   ]
  }
  MaterialExpressionLinearInterpolate_5 {
   class=MaterialExpressionLinearInterpolate
   inputs [
    A {from=MaterialExpressionLinearInterpolate_4}
    B {from=MaterialExpressionNamedRerouteUsage_5}
    Alpha {from=MaterialExpressionNamedRerouteUsage_25}
   ]
   properties {constB=0.1763, desc=Detail UV Scale}
  }
  MaterialExpressionMaterialFunctionCall_23 {
   class=MaterialExpressionMaterialFunctionCall
   inputs []
   properties {materialFunction=&TangentBasis}
  }
  MaterialExpressionMaterialFunctionCall_25 {
   class=MaterialExpressionMaterialFunctionCall
   inputs [
    Normal {from=MaterialExpressionTextureSample_14, output=RGB}
    Intensity {from=MaterialExpressionLinearInterpolate_30}
   ]
   properties {materialFunction=/Game/Global/MaterialFunctions/MF_NormalIntensity}
  }
  MaterialExpressionMaterialFunctionCall_26 {
   class=MaterialExpressionMaterialFunctionCall
   inputs [
    View Vector {from=MaterialExpressionNamedRerouteUsage_421}
    Normal {from=MaterialExpressionNamedRerouteUsage_373}
    "Relative IOR (Eta)" {from=MaterialExpressionDivide_27}
   ]
   properties {materialFunction=/Game/Global/MaterialFunctions/TechArt/Refraction/MF_Refract}
  }
  MaterialExpressionMaterialFunctionCall_28 {
   class=MaterialExpressionMaterialFunctionCall
   inputs [
    Direction WS {from=MaterialExpressionNamedRerouteUsage_376}
    Roughness {from=MaterialExpressionNamedRerouteUsage_377}
    Sky Desaturation {from=MaterialExpressionNamedRerouteUsage_538}
    Sky Brightness {from=""}
    Use Ground Map {from=MaterialExpressionStaticBool_2}
    Ground LatLong Map {from=MaterialExpressionTextureObject_3}
    Group Map Rotation {from=""}
    Ground Exp {from=MaterialExpressionConstant_30}
    Ground Brightness {from=MaterialExpressionConstant_31}
   ]
   properties {materialFunction=/Game/Global/MaterialFunctions/TechArt/Refraction/MF_EnvironmentIlluminance}
  }
  MaterialExpressionMaterialFunctionCall_29 {
   class=MaterialExpressionMaterialFunctionCall
   inputs [
    Normal WS {from=MaterialExpressionNamedRerouteUsage_397}
    Roughness {from=MaterialExpressionNamedRerouteUsage_394}
    Surface Opacity {from=MaterialExpressionNamedRerouteUsage_486}
    Refraction Vector WS {from=MaterialExpressionNamedRerouteUsage_393}
    Refraction IOR {from=MaterialExpressionNamedRerouteUsage_395}
    Scene Luminance {from=MaterialExpressionNamedRerouteUsage_396}
    Interior Emission {from=MaterialExpressionNamedRerouteUsage_399}
    Volume Thickness {from=MaterialExpressionNamedRerouteUsage_398}
    Volume Density {from=MaterialExpressionNamedRerouteUsage_416}
    Absorption Coeffs {from=MaterialExpressionNamedRerouteUsage_414}
    Scattering Coeffs {from=MaterialExpressionNamedRerouteUsage_415}
    Scattering Phase {from=MaterialExpressionNamedRerouteUsage_418}
    Is Scene Inside Volume {from=MaterialExpressionConstant_35}
    Thin Transmission Exp {from=MaterialExpressionNamedRerouteUsage_417}
    //---- ALPHA-BLENDING ONLY ----// {from=""}
    Haze Amount {from=""}
    //----------- LIGHTING -----------// {from=""}
    Sky Illuminance {from=MaterialExpressionNamedRerouteUsage_478}
    Sun Illuminance {from=MaterialExpressionNamedRerouteUsage_479}
    Sun Direction {from=MaterialExpressionNamedRerouteUsage_480}
   ]
   properties {materialFunction=/Game/Global/MaterialFunctions/TechArt/Refraction/MF_DielectricInteriorVolume}
  }
  MaterialExpressionMaterialFunctionCall_30 {
   class=MaterialExpressionMaterialFunctionCall
   inputs [
    Normal_1 {from=MaterialExpressionNamedRerouteUsage_412}
    Normal_2 {from=MaterialExpressionMaterialFunctionCall_25, output="Result (V3)"}
   ]
   properties {materialFunction=/Game/Global/MaterialFunctions/MF_NormalBlend_Whiteout}
  }
  MaterialExpressionMaterialFunctionCall_31 {
   class=MaterialExpressionMaterialFunctionCall
   inputs [
    Color_Morning {from=MaterialExpressionVectorParameter_54, output=RGB}
    Color_Day {from=MaterialExpressionVectorParameter_55, output=RGB}
    Color_Evening {from=MaterialExpressionVectorParameter_56, output=RGB}
    Color_Night {from=MaterialExpressionVectorParameter_57, output=RGB}
   ]
   properties {materialFunction=&Color_Time_of_Day}
  }
  MaterialExpressionMaterialFunctionCall_32 {
   class=MaterialExpressionMaterialFunctionCall
   inputs [
    Normal WS {from=MaterialExpressionNamedRerouteUsage_477}
    Sky Illuminance Mult {from=""}
    Sky Desaturation {from=MaterialExpressionConstant_32}
    Sun Illuminance Mult {from=""}
    Sun Desaturation {from=MaterialExpressionConstant_32}
   ]
   properties {materialFunction=/Game/Global/MaterialFunctions/TechArt/Refraction/MF_SunSkyProperties}
  }
  MaterialExpressionMaterialFunctionCall_36 {
   class=MaterialExpressionMaterialFunctionCall
   inputs [
    Normal {from=MaterialExpressionNamedRerouteUsage_482}
    Direction {from=MaterialExpressionNamedRerouteUsage_374}
    Facing Ratio Exp {from=MaterialExpressionConstant_38}
    Parallel Intensity {from=MaterialExpressionConstant_39}
    Perpendicular Intensity {from=MaterialExpressionConstant_40}
   ]
   properties {materialFunction=/Game/Global/MaterialFunctions/TechArt/Refraction/MF_ThicknessSimple}
  }
  MaterialExpressionMaterialFunctionCall_39 {
   class=MaterialExpressionMaterialFunctionCall
   inputs [
    In {from=MaterialExpressionTextureSampleParameter2D_2, output=RGB}
    Contrast {from=MaterialExpressionNamedRerouteUsage_539}
   ]
   properties {materialFunction=/Engine/Functions/Engine_MaterialFunctions01/ImageAdjustment/CheapContrast_RGB}
  }
  MaterialExpressionMaterialFunctionCall_40 {
   class=MaterialExpressionMaterialFunctionCall
   inputs [
    Noise Texture {from=MaterialExpressionReroute_24}
    Triplanar Tiling {from=MaterialExpressionReroute_25}
    Triplanar Tiling V Ratio {from=MaterialExpressionReroute_26}
    Triplanar Blend Contrast {from=MaterialExpressionReroute_27}
    Variation Noise Settings {from=MaterialExpressionReroute_28}
    Noise Depth Breakup Settings {from=MaterialExpressionReroute_29}
    Time {from=MaterialExpressionReroute_30}
    Refraction Vector WS {from=MaterialExpressionReroute_31}
    Parallax Depth Norm {from=MaterialExpressionReroute_32}
    Projection Min Cos {from=MaterialExpressionReroute_33}
    Grazing Angle Parallax Mult {from=MaterialExpressionReroute_34}
    Grazing Angle Flatten Min Cos {from=MaterialExpressionReroute_35}
   ]
   properties {materialFunction=&/MF_RefractionDepthVariation}
  }
  MaterialExpressionMax_14 {
   class=MaterialExpressionMax
   inputs [
    A {from=MaterialExpressionMultiply_90}
    B {from=""}
   ]
   properties {constB=0.001}
  }
  MaterialExpressionMax_15 {
   class=MaterialExpressionMax
   inputs [
    A {from=MaterialExpressionScalarParameter_66}
    B {from=""}
   ]
   properties {constB=0}
  }
  MaterialExpressionMax_16 {
   class=MaterialExpressionMax
   inputs [
    A {from=MaterialExpressionComponentMask_7}
    B {from=""}
   ]
   properties {constB=0.001}
  }
  MaterialExpressionMax_17 {
   class=MaterialExpressionMax
   inputs [
    A {from=MaterialExpressionMultiply_86}
    B {from=""}
   ]
   properties {constB=0.0001}
  }
  MaterialExpressionMax_18 {
   class=MaterialExpressionMax
   inputs [
    A {from=MaterialExpressionNamedRerouteUsage_409}
    B {from=""}
   ]
   properties {constB=0}
  }
  MaterialExpressionMultiply_0 {
   class=MaterialExpressionMultiply
   inputs [
    A {from=MaterialExpressionOneMinus_0}
    B {from=MaterialExpressionStep_1}
   ]
  }
  MaterialExpressionMultiply_1 {
   class=MaterialExpressionMultiply
   inputs [
    A {from=MaterialExpressionOneMinus_1}
    B {from=MaterialExpressionStep_9}
   ]
  }
  MaterialExpressionMultiply_101 {
   class=MaterialExpressionMultiply
   inputs [
    A {from=MaterialExpressionPower_39}
    B {from=MaterialExpressionNamedRerouteUsage_497}
   ]
  }
  MaterialExpressionMultiply_102 {
   class=MaterialExpressionMultiply
   inputs [
    A {from=MaterialExpressionPower_40}
    B {from=MaterialExpressionNamedRerouteUsage_494}
   ]
  }
  MaterialExpressionMultiply_105 {
   class=MaterialExpressionMultiply
   inputs [
    A {from=MaterialExpressionNamedRerouteUsage_503}
    B {from=MaterialExpressionMaterialFunctionCall_29, output="Refraction Result - Opaque (V3)"}
   ]
  }
  MaterialExpressionMultiply_111 {
   class=MaterialExpressionMultiply
   inputs [
    A {from=MaterialExpressionStep_2}
    B {from=MaterialExpressionOneMinus_11}
   ]
  }
  MaterialExpressionMultiply_112 {
   class=MaterialExpressionMultiply
   inputs [
    A {from=MaterialExpressionConvert_31}
    B {from=MaterialExpressionConvert_31}
   ]
  }
  MaterialExpressionMultiply_113 {
   class=MaterialExpressionMultiply
   inputs [
    A {from=MaterialExpressionFrac_5}
    B {from=MaterialExpressionNamedRerouteUsage_548}
   ]
  }
  MaterialExpressionMultiply_114 {
   class=MaterialExpressionMultiply
   inputs [
    A {from=MaterialExpressionNamedRerouteUsage_558}
    B {from=MaterialExpressionFrac_6}
   ]
  }
  MaterialExpressionMultiply_2 {
   class=MaterialExpressionMultiply
   inputs [
    A {from=MaterialExpressionOneMinus_2}
    B {from=MaterialExpressionStep_11}
   ]
  }
  MaterialExpressionMultiply_3 {
   class=MaterialExpressionMultiply
   inputs [
    A {from=MaterialExpressionOneMinus_3}
    B {from=MaterialExpressionStep_13}
   ]
  }
  MaterialExpressionMultiply_77 {
   class=MaterialExpressionMultiply
   inputs [
    A {from=MaterialExpressionPower_28}
    B {from=MaterialExpressionNamedRerouteUsage_410}
   ]
  }
  MaterialExpressionMultiply_78 {
   class=MaterialExpressionMultiply
   inputs [
    A {from=MaterialExpressionMultiply_80}
    B {from=MaterialExpressionNamedRerouteUsage_445}
   ]
  }
  MaterialExpressionMultiply_79 {
   class=MaterialExpressionMultiply
   inputs [
    A {from=MaterialExpressionDivide_29}
    B {from=MaterialExpressionNamedRerouteUsage_379}
   ]
  }
  MaterialExpressionMultiply_80 {
   class=MaterialExpressionMultiply
   inputs [
    A {from=MaterialExpressionPower_29}
    B {from=MaterialExpressionNamedRerouteUsage_384}
   ]
  }
  MaterialExpressionMultiply_81 {
   class=MaterialExpressionMultiply
   inputs [
    A {from=MaterialExpressionNamedRerouteUsage_426}
    B {from=MaterialExpressionPower_30}
   ]
  }
  MaterialExpressionMultiply_84 {
   class=MaterialExpressionMultiply
   inputs [
    A {from=MaterialExpressionNamedRerouteUsage_390}
    B {from=MaterialExpressionDivide_33}
   ]
  }
  MaterialExpressionMultiply_85 {
   class=MaterialExpressionMultiply
   inputs [
    A {from=MaterialExpressionConvert_22}
    B {from=MaterialExpressionNamedRerouteUsage_433}
   ]
  }
  MaterialExpressionMultiply_86 {
   class=MaterialExpressionMultiply
   inputs [
    A {from=MaterialExpressionConvert_22}
    B {from=MaterialExpressionNamedRerouteUsage_434}
   ]
  }
  MaterialExpressionMultiply_87 {
   class=MaterialExpressionMultiply
   inputs [
    A {from=MaterialExpressionNamedRerouteUsage_413}
    B {from=MaterialExpressionNamedRerouteUsage_428}
   ]
  }
  MaterialExpressionMultiply_88 {
   class=MaterialExpressionMultiply
   inputs [
    A {from=MaterialExpressionNamedRerouteUsage_449}
    B {from=MaterialExpressionLinearInterpolate_31}
   ]
  }
  MaterialExpressionMultiply_89 {
   class=MaterialExpressionMultiply
   inputs [
    A {from=MaterialExpressionTime_2}
    B {from=MaterialExpressionNamedRerouteUsage_476}
   ]
  }
  MaterialExpressionMultiply_90 {
   class=MaterialExpressionMultiply
   inputs [
    A {from=MaterialExpressionComponentMask_8}
    B {from=""}
   ]
   properties {constB=-1, desc=cosTheta}
  }
  MaterialExpressionMultiply_91 {
   class=MaterialExpressionMultiply
   inputs [
    A {from=MaterialExpressionMaterialFunctionCall_28, output="Environment Illuminance (V3)"}
    B {from=MaterialExpressionMax_18}
   ]
  }
  MaterialExpressionMultiply_92 {
   class=MaterialExpressionMultiply
   inputs [
    A {from=MaterialExpressionVectorParameter_44, output=RGB}
    B {from=MaterialExpressionVectorParameter_44, output=RGB}
   ]
  }
  MaterialExpressionMultiply_93 {
   class=MaterialExpressionMultiply
   inputs [
    A {from=MaterialExpressionPower_35}
    B {from=MaterialExpressionNamedRerouteUsage_447}
   ]
  }
  MaterialExpressionMultiply_94 {
   class=MaterialExpressionMultiply
   inputs [
    A {from=MaterialExpressionNamedRerouteUsage_406}
    B {from=MaterialExpressionNamedRerouteUsage_406}
   ]
  }
  MaterialExpressionMultiply_95 {
   class=MaterialExpressionMultiply
   inputs [
    A {from=MaterialExpressionMultiply_89}
    B {from=""}
   ]
   properties {constB=0.004}
  }
  MaterialExpressionNamedRerouteDeclaration_0 {
   class=MaterialExpressionNamedRerouteDeclaration
   inputs [None {from=MaterialExpressionScalarParameter_0}]
   properties {name="Detail UV Scale - Base (S)", nodeColor=<rgba 0.4766 1 0 1>}
  }
  MaterialExpressionNamedRerouteDeclaration_1 {
   class=MaterialExpressionNamedRerouteDeclaration
   inputs [None {from=MaterialExpressionVectorParameter_0, output=Rod}]
   properties {name="Detail UV Scale - Rod (S)", nodeColor=<rgba 0.4766 1 0 1>}
  }
  MaterialExpressionNamedRerouteDeclaration_13 {
   class=MaterialExpressionNamedRerouteDeclaration
   inputs [None {from=MaterialExpressionSwitch_1}]
   properties {name="Mask Top Ring (S)", nodeColor=<rgba 1 0.2109 0 1>}
  }
  MaterialExpressionNamedRerouteDeclaration_14 {
   class=MaterialExpressionNamedRerouteDeclaration
   inputs [None {from=MaterialExpressionSwitch_3}]
   properties {name="Mask Bottom Ring (S)", nodeColor=<rgba 1 0.2109 0 1>}
  }
  MaterialExpressionNamedRerouteDeclaration_15 {
   class=MaterialExpressionNamedRerouteDeclaration
   inputs [None {from=MaterialExpressionSwitch_0}]
   properties {name="Mask Rod (S)", nodeColor=<rgba 1 0.2109 0 1>}
  }
  MaterialExpressionNamedRerouteDeclaration_16 {
   class=MaterialExpressionNamedRerouteDeclaration
   inputs [None {from=MaterialExpressionSwitch_2}]
   properties {name="Mask Middle Ring (S)", nodeColor=<rgba 1 0.2109 0 1>}
  }
  MaterialExpressionNamedRerouteDeclaration_2 {
   class=MaterialExpressionNamedRerouteDeclaration
   inputs [None {from=MaterialExpressionVectorParameter_0, output=Top Ring}]
   properties {name="Detail UV Scale - Top Ring (S)", nodeColor=<rgba 0.4766 1 0 1>}
  }
  MaterialExpressionNamedRerouteDeclaration_20 {
   class=MaterialExpressionNamedRerouteDeclaration
   inputs [None {from=MaterialExpressionChannelMaskParameter_1, output=R}]
   properties {name="Mask Rod Channel (S)", nodeColor=<rgba 0 0.2266 1 1>}
  }
  MaterialExpressionNamedRerouteDeclaration_21 {
   class=MaterialExpressionNamedRerouteDeclaration
   inputs [None {from=MaterialExpressionChannelMaskParameter_2, output=R}]
   properties {name="Mask Top Ring Channel (S)", nodeColor=<rgba 0 0.2266 1 1>}
  }
  MaterialExpressionNamedRerouteDeclaration_22 {
   class=MaterialExpressionNamedRerouteDeclaration
   inputs [None {from=MaterialExpressionChannelMaskParameter_3, output=R}]
   properties {name="Mask Middle Ring Channel (S)", nodeColor=<rgba 0 0.2266 1 1>}
  }
  MaterialExpressionNamedRerouteDeclaration_23 {
   class=MaterialExpressionNamedRerouteDeclaration
   inputs [None {from=MaterialExpressionChannelMaskParameter_8, output=R}]
   properties {name="Mask Bottom Ring Channel (S)", nodeColor=<rgba 0 0.2266 1 1>}
  }
  MaterialExpressionNamedRerouteDeclaration_24 {
   class=MaterialExpressionNamedRerouteDeclaration
   inputs [None {from=MaterialExpressionRound_0}]
   properties {name="Region Masks - Is Low (V4)", nodeColor=<rgba 0.0312 1 0 1>}
  }
  MaterialExpressionNamedRerouteDeclaration_264 {
   class=MaterialExpressionNamedRerouteDeclaration
   inputs [None {from=MaterialExpressionMultiply_77}]
   properties {name="Geometric Thickness (S)", nodeColor=<rgba 0 1 0.6484 1>}
  }
  MaterialExpressionNamedRerouteDeclaration_265 {
   class=MaterialExpressionNamedRerouteDeclaration
   inputs [None {from=MaterialExpressionDivide_26}]
   properties {name="View Dependent Thickness (S)", nodeColor=<rgba 1 0 0.5156 1>}
  }
  MaterialExpressionNamedRerouteDeclaration_266 {
   class=MaterialExpressionNamedRerouteDeclaration
   inputs [None {from=MaterialExpressionMaterialFunctionCall_26, output="Refraction Vector (V3)"}]
   properties {name="Refraction Vector WS (V3)", nodeColor=<rgba 0.5469 1 0 1>}
  }
  MaterialExpressionNamedRerouteDeclaration_267 {
   class=MaterialExpressionNamedRerouteDeclaration
   inputs [None {from=MaterialExpressionClamp_2}]
   properties {name="Refraction - IOR (S)", nodeColor=<rgba 1 0.0938 0 1>}
  }
  MaterialExpressionNamedRerouteDeclaration_268 {
   class=MaterialExpressionNamedRerouteDeclaration
   inputs [None {from=MaterialExpressionNormalize_2}]
   properties {name="Surface Normal WS (V3)", nodeColor=<rgba 0 0.5913 1 1>}
  }
  MaterialExpressionNamedRerouteDeclaration_269 {
   class=MaterialExpressionNamedRerouteDeclaration
   inputs [None {from=MaterialExpressionScalarParameter_68}]
   properties {name="Refraction Roughness (S)", nodeColor=<rgba 0.1719 0 1 1>}
  }
  MaterialExpressionNamedRerouteDeclaration_270 {
   class=MaterialExpressionNamedRerouteDeclaration
   inputs [None {from=MaterialExpressionMultiply_91}]
   properties {name="Refracted Illuminance (V3)", nodeColor=<rgba 0.9688 0 1 1>}
  }
  MaterialExpressionNamedRerouteDeclaration_271 {
   class=MaterialExpressionNamedRerouteDeclaration
   inputs [None {from=MaterialExpressionAdd_38}]
   properties {name="Refracted Emission (V3)", nodeColor=<rgba 0 1 0.6484 1>}
  }
  MaterialExpressionNamedRerouteDeclaration_272 {
   class=MaterialExpressionNamedRerouteDeclaration
   inputs [None {from=MaterialExpressionSetMaterialAttributes_4}]
   properties {name="Material Out (MA)", nodeColor=<rgba 0.3125 1 0 1>}
  }
  MaterialExpressionNamedRerouteDeclaration_273 {
   class=MaterialExpressionNamedRerouteDeclaration
   inputs [None {from=MaterialExpressionConvert_21}]
   properties {name="Refraction Vector TS (V3)", nodeColor=<rgba 0 1 0.3203 1>}
  }
  MaterialExpressionNamedRerouteDeclaration_274 {
   class=MaterialExpressionNamedRerouteDeclaration
   inputs [None {from=MaterialExpressionDivide_28}]
   properties {name="Refracted Unit UV Offset TS (V2)", nodeColor=<rgba 1 0 0.1875 1>}
  }
  MaterialExpressionNamedRerouteDeclaration_275 {
   class=MaterialExpressionNamedRerouteDeclaration
   inputs [None {from=MaterialExpressionHsvToRgb_2}]
   properties {name="Volume Absorption Coeffs (V3)", nodeColor=<rgba 1 0.7266 0 1>}
  }
  MaterialExpressionNamedRerouteDeclaration_276 {
   class=MaterialExpressionNamedRerouteDeclaration
   inputs [None {from=MaterialExpressionMultiply_92}]
   properties {name="Volume Scattering Coeffs (V3)", nodeColor=<rgba 0.4062 0 1 1>}
  }
  MaterialExpressionNamedRerouteDeclaration_277 {
   class=MaterialExpressionNamedRerouteDeclaration
   inputs [None {from=MaterialExpressionTextureCoordinate_2}]
   properties {name="UV Asset (V2)", nodeColor=<rgba 0.9219 0 1 1>}
  }
  MaterialExpressionNamedRerouteDeclaration_279 {
   class=MaterialExpressionNamedRerouteDeclaration
   inputs [None {from=MaterialExpressionMaterialFunctionCall_40, output="Result (S)"}]
   properties {name="Refraction Depth Variation (S)", nodeColor=<rgba 1 0 0 1>}
  }
  MaterialExpressionNamedRerouteDeclaration_280 {
   class=MaterialExpressionNamedRerouteDeclaration
   inputs [None {from=MaterialExpressionConstant_34}]
   properties {name="Parallax Depth Norm (S)", nodeColor=<rgba 0 0.8594 1 1>}
  }
  MaterialExpressionNamedRerouteDeclaration_281 {
   class=MaterialExpressionNamedRerouteDeclaration
   inputs [None {from=MaterialExpressionAdd_37}]
   properties {name="Emission Out (V3)", nodeColor=<rgba 0.7578 1 0 1>}
  }
  MaterialExpressionNamedRerouteDeclaration_282 {
   class=MaterialExpressionNamedRerouteDeclaration
   inputs [None {from=MaterialExpressionMultiply_87}]
   properties {name="Base Color Out (V3)", nodeColor=<rgba 0 0.9297 1 1>}
  }
  MaterialExpressionNamedRerouteDeclaration_283 {
   class=MaterialExpressionNamedRerouteDeclaration
   inputs [None {from=MaterialExpressionNamedRerouteUsage_442}]
   properties {name="Specular Out (S)", nodeColor=<rgba 1 0.75 0 1>}
  }
  MaterialExpressionNamedRerouteDeclaration_284 {
   class=MaterialExpressionNamedRerouteDeclaration
   inputs [None {from=MaterialExpressionNamedRerouteUsage_441}]
   properties {name="Roughness Out (S)", nodeColor=<rgba 0.4297 0 1 1>}
  }
  MaterialExpressionNamedRerouteDeclaration_285 {
   class=MaterialExpressionNamedRerouteDeclaration
   inputs [None {from=MaterialExpressionDivide_35}]
   properties {name="Ice Surface UV (V2)", nodeColor=<rgba 1 0 0.2109 1>}
  }
  MaterialExpressionNamedRerouteDeclaration_286 {
   class=MaterialExpressionNamedRerouteDeclaration
   inputs [None {from=MaterialExpressionNamedRerouteUsage_440}]
   properties {name="Normal Out TS (V3)", nodeColor=<rgba 1 0.1875 0 1>}
  }
  MaterialExpressionNamedRerouteDeclaration_287 {
   class=MaterialExpressionNamedRerouteDeclaration
   inputs [None {from=MaterialExpressionMaterialFunctionCall_36, output="Thickness Normalized (S)"}]
   properties {name="Normalized Thickness (S)", nodeColor=<rgba 0 0.6953 1 1>}
  }
  MaterialExpressionNamedRerouteDeclaration_288 {
   class=MaterialExpressionNamedRerouteDeclaration
   inputs [None {from=MaterialExpressionScalarParameter_79}]
   properties {name=Refraction Intensity, nodeColor=<rgba 0 1 0.6953 1>}
  }
  MaterialExpressionNamedRerouteDeclaration_289 {
   class=MaterialExpressionNamedRerouteDeclaration
   inputs [None {from=MaterialExpressionScalarParameter_65}]
   properties {name="Thickness Exp (S)", nodeColor=<rgba 1 0.375 0 1>}
  }
  MaterialExpressionNamedRerouteDeclaration_290 {
   class=MaterialExpressionNamedRerouteDeclaration
   inputs [None {from=MaterialExpressionMax_15}]
   properties {name="Max Thickness (S)", nodeColor=<rgba 0.0547 0 1 1>}
  }
  MaterialExpressionNamedRerouteDeclaration_291 {
   class=MaterialExpressionNamedRerouteDeclaration
   inputs [None {from=MaterialExpressionScalarParameter_80}]
   properties {name="Volume Density (S)", nodeColor=<rgba 0 1 0.9531 1>}
  }
  MaterialExpressionNamedRerouteDeclaration_292 {
   class=MaterialExpressionNamedRerouteDeclaration
   inputs [None {from=MaterialExpressionScalarParameter_82}]
   properties {name="Transmission Attenuation (S)", nodeColor=<rgba 1 0.6328 0 1>}
  }
  MaterialExpressionNamedRerouteDeclaration_293 {
   class=MaterialExpressionNamedRerouteDeclaration
   inputs [None {from=MaterialExpressionScalarParameter_81}]
   properties {name="Volume Scattering Phase (S)", nodeColor=<rgba 0.3125 0 1 1>}
  }
  MaterialExpressionNamedRerouteDeclaration_294 {
   class=MaterialExpressionNamedRerouteDeclaration
   inputs [None {from=MaterialExpressionMultiply_95}]
   properties {name="Time (S)", nodeColor=<rgba 0 0.7656 1 1>}
  }
  MaterialExpressionNamedRerouteDeclaration_295 {
   class=MaterialExpressionNamedRerouteDeclaration
   inputs [None {from=MaterialExpressionTextureCoordinate_2}]
   properties {name="UV Tiles (V2)", nodeColor=<rgba 1 0.5859 0 1>}
  }
  MaterialExpressionNamedRerouteDeclaration_296 {
   class=MaterialExpressionNamedRerouteDeclaration
   inputs [None {from=MaterialExpressionCameraVectorWS_1}]
   properties {name="Camera Vector WS (V3)", nodeColor=<rgba 1 0 0 1>}
  }
  MaterialExpressionNamedRerouteDeclaration_297 {
   class=MaterialExpressionNamedRerouteDeclaration
   inputs [None {from=MaterialExpressionMaterialFunctionCall_39, output=Result}]
   properties {name="Diffuse Map Lookup (V3)", nodeColor=<rgba 0.6172 1 0 1>}
  }
  MaterialExpressionNamedRerouteDeclaration_298 {
   class=MaterialExpressionNamedRerouteDeclaration
   inputs [None {from=MaterialExpressionVectorParameter_45, output=RGB}]
   properties {name="Emissive Fresnel Tint (V3)", nodeColor=<rgba 0 1 0.2734 1>}
  }
  MaterialExpressionNamedRerouteDeclaration_299 {
   class=MaterialExpressionNamedRerouteDeclaration
   inputs [None {from=MaterialExpressionVectorParameter_45, output=Fresnel Exp}]
   properties {name="Emissive Fresnel Exp (S)", nodeColor=<rgba 0.6406 1 0 1>}
  }
  MaterialExpressionNamedRerouteDeclaration_3 {
   class=MaterialExpressionNamedRerouteDeclaration
   inputs [None {from=MaterialExpressionVectorParameter_0, output=Middle Ring}]
   properties {name="Detail UV Scale - Middle Ring (S)", nodeColor=<rgba 0.4766 1 0 1>}
  }
  MaterialExpressionNamedRerouteDeclaration_300 {
   class=MaterialExpressionNamedRerouteDeclaration
   inputs [None {from=MaterialExpressionSaturate_29}]
   properties {name="Facing Ratio VtxN (S)", nodeColor=<rgba 0 1 0.6484 1>}
  }
  MaterialExpressionNamedRerouteDeclaration_301 {
   class=MaterialExpressionNamedRerouteDeclaration
   inputs [None {from=MaterialExpressionVectorParameter_46, output=Hue Shift}]
   properties {name="Diffuse Map Hue Shift (S)", nodeColor=<rgba 0.1953 1 0 1>}
  }
  MaterialExpressionNamedRerouteDeclaration_302 {
   class=MaterialExpressionNamedRerouteDeclaration
   inputs [None {from=MaterialExpressionVectorParameter_46, output=Saturation}]
   properties {name="Diffuse Map Saturation (S)", nodeColor=<rgba 1 0 0.5156 1>}
  }
  MaterialExpressionNamedRerouteDeclaration_303 {
   class=MaterialExpressionNamedRerouteDeclaration
   inputs [None {from=MaterialExpressionVectorParameter_46, output=Value}]
   properties {name="Diffuse Map Value (S)", nodeColor=<rgba 0 0.8359 1 1>}
  }
  MaterialExpressionNamedRerouteDeclaration_304 {
   class=MaterialExpressionNamedRerouteDeclaration
   inputs [None {from=MaterialExpressionVectorParameter_46, output=Diffuse Map Exp}]
   properties {name="Diffuse Map Exp (S)", nodeColor=<rgba 1 0.8438 0 1>}
  }
  MaterialExpressionNamedRerouteDeclaration_305 {
   class=MaterialExpressionNamedRerouteDeclaration
   inputs [None {from=MaterialExpressionVectorParameter_47, output=Fresnel Exp}]
   properties {name="Detail Normal Fresnel Exp (S)", nodeColor=<rgba 1 0.0703 0 1>}
  }
  MaterialExpressionNamedRerouteDeclaration_306 {
   class=MaterialExpressionNamedRerouteDeclaration
   inputs [None {from=MaterialExpressionVectorParameter_47, output="Max Intensity (Grazing)"}]
   properties {name="Detail Normal Max Intensity (S)", nodeColor=<rgba 0 0.25 1 1>}
  }
  MaterialExpressionNamedRerouteDeclaration_307 {
   class=MaterialExpressionNamedRerouteDeclaration
   inputs [None {from=MaterialExpressionVectorParameter_47, output="Min Intensity (Facing)"}]
   properties {name="Detail Normal Min Intensity (S)", nodeColor=<rgba 0.5703 1 0 1>}
  }
  MaterialExpressionNamedRerouteDeclaration_308 {
   class=MaterialExpressionNamedRerouteDeclaration
   inputs [None {from=MaterialExpressionMaterialFunctionCall_30, output=Result}]
   properties {name="Surface Normal TS (V3)", nodeColor=<rgba 0 0.8594 1 1>}
  }
  MaterialExpressionNamedRerouteDeclaration_309 {
   class=MaterialExpressionNamedRerouteDeclaration
   inputs [None {from=MaterialExpressionLinearInterpolate_36}]
   properties {name="Ice Roughness (S)", nodeColor=<rgba 0 0.4844 1 1>}
  }
  MaterialExpressionNamedRerouteDeclaration_310 {
   class=MaterialExpressionNamedRerouteDeclaration
   inputs [None {from=MaterialExpressionSaturate_36}]
   properties {name="Ice Specular (S)", nodeColor=<rgba 0.8047 1 0 1>}
  }
  MaterialExpressionNamedRerouteDeclaration_311 {
   class=MaterialExpressionNamedRerouteDeclaration
   inputs [None {from=MaterialExpressionMultiply_105}]
   properties {name="Refraction Luminance (S)", nodeColor=<rgba 1 0 0.75 1>}
  }
  MaterialExpressionNamedRerouteDeclaration_312 {
   class=MaterialExpressionNamedRerouteDeclaration
   inputs [None {from=MaterialExpressionScalarParameter_70}]
   properties {name="Interior Emission Exp (S)", nodeColor=<rgba 0 0.6953 1 1>}
  }
  MaterialExpressionNamedRerouteDeclaration_313 {
   class=MaterialExpressionNamedRerouteDeclaration
   inputs [None {from=MaterialExpressionScalarParameter_69}]
   properties {name="Interior Emission Intensity (S)", nodeColor=<rgba 0 0.6953 1 1>}
  }
  MaterialExpressionNamedRerouteDeclaration_314 {
   class=MaterialExpressionNamedRerouteDeclaration
   inputs [None {from=MaterialExpressionVectorParameter_52, output=Roughness Exp}]
   properties {name="Detail Roughness Exp (S)", nodeColor=<rgba 1 0.9844 0 1>}
  }
  MaterialExpressionNamedRerouteDeclaration_315 {
   class=MaterialExpressionNamedRerouteDeclaration
   inputs [None {from=MaterialExpressionVectorParameter_52, output=Roughness Mult}]
   properties {name="Detail Roughness Mult (S)", nodeColor=<rgba 0.6641 0 1 1>}
  }
  MaterialExpressionNamedRerouteDeclaration_317 {
   class=MaterialExpressionNamedRerouteDeclaration
   inputs [None {from=MaterialExpressionScalarParameter_72}]
   properties {name="Diffuse Depth (S)", nodeColor=<rgba 0.0781 1 0 1>}
  }
  MaterialExpressionNamedRerouteDeclaration_318 {
   class=MaterialExpressionNamedRerouteDeclaration
   inputs [None {from=MaterialExpressionScalarParameter_74}]
   properties {name="Diffuse Depth Attenuation (S)", nodeColor=<rgba 1 0 0.3984 1>}
  }
  MaterialExpressionNamedRerouteDeclaration_319 {
   class=MaterialExpressionNamedRerouteDeclaration
   inputs [None {from=MaterialExpressionTextureObjectParameter_5}]
   properties {name="Surface Detail Roughness Map (TO)", nodeColor=<rgba 0.125 1 0 1>}
  }
  MaterialExpressionNamedRerouteDeclaration_320 {
   class=MaterialExpressionNamedRerouteDeclaration
   inputs [None {from=MaterialExpressionTextureObjectParameter_6}]
   properties {name="Surface Detail Normal Map (TO)", nodeColor=<rgba 0.6406 1 0 1>}
  }
  MaterialExpressionNamedRerouteDeclaration_321 {
   class=MaterialExpressionNamedRerouteDeclaration
   inputs [None {from=MaterialExpressionScalarParameter_71}]
   properties {name="Emission Depth (S)", nodeColor=<rgba 0.5703 1 0 1>}
  }
  MaterialExpressionNamedRerouteDeclaration_322 {
   class=MaterialExpressionNamedRerouteDeclaration
   inputs [None {from=MaterialExpressionNamedRerouteUsage_456}]
   properties {name="Mask Out (S)", nodeColor=<rgba 1 0.3984 0 1>}
  }
  MaterialExpressionNamedRerouteDeclaration_339 {
   class=MaterialExpressionNamedRerouteDeclaration
   inputs [None {from=MaterialExpressionScalarParameter_75}]
   properties {name="Interior Noise Speed (S)", nodeColor=<rgba 0 1 0.0625 1>}
  }
  MaterialExpressionNamedRerouteDeclaration_340 {
   class=MaterialExpressionNamedRerouteDeclaration
   inputs [None {from=MaterialExpressionMaterialFunctionCall_31, output=Result}]
   properties {name="TOD Color (V3)", nodeColor=<rgba 0.5938 0 1 1>}
  }
  MaterialExpressionNamedRerouteDeclaration_341 {
   class=MaterialExpressionNamedRerouteDeclaration
   inputs [None {from=MaterialExpressionMaterialFunctionCall_32, output="Sun Direction (V3)"}]
   properties {name="Lighting - Sun Direction (V3)", nodeColor=<rgba 0 0.0625 1 1>}
  }
  MaterialExpressionNamedRerouteDeclaration_342 {
   class=MaterialExpressionNamedRerouteDeclaration
   inputs [None {from=MaterialExpressionMaterialFunctionCall_32, output="Sun Illuminance (V3)"}]
   properties {name="Lighting - Sun Illuminance (V3)", nodeColor=<rgba 0.7344 0 1 1>}
  }
  MaterialExpressionNamedRerouteDeclaration_343 {
   class=MaterialExpressionNamedRerouteDeclaration
   inputs [None {from=MaterialExpressionMaterialFunctionCall_32, output="Sky Illuminance (V3)"}]
   properties {name="Lighting - Sky Illuminance (V3)", nodeColor=<rgba 0 1 0.4141 1>}
  }
  MaterialExpressionNamedRerouteDeclaration_344 {
   class=MaterialExpressionNamedRerouteDeclaration
   inputs [None {from=MaterialExpressionMultiply_81}]
   properties {name="Emissive Fresnel (V3)", nodeColor=<rgba 1 0 0.4688 1>}
  }
  MaterialExpressionNamedRerouteDeclaration_348 {
   class=MaterialExpressionNamedRerouteDeclaration
   inputs [None {from=MaterialExpressionVectorParameter_52, output=Mix Amount}]
   properties {name="Detail Roughness Mix (S)", nodeColor=<rgba 0 0.9297 1 1>}
  }
  MaterialExpressionNamedRerouteDeclaration_349 {
   class=MaterialExpressionNamedRerouteDeclaration
   inputs [None {from=MaterialExpressionVectorParameter_53, output=Roughness Exp}]
   properties {name="Asset Roughness Exp (S)", nodeColor=<rgba 1 0.75 0 1>}
  }
  MaterialExpressionNamedRerouteDeclaration_350 {
   class=MaterialExpressionNamedRerouteDeclaration
   inputs [None {from=MaterialExpressionVectorParameter_53, output=Roughness Mult}]
   properties {name="Asset Roughness Mult (S)", nodeColor=<rgba 0.4297 0 1 1>}
  }
  MaterialExpressionNamedRerouteDeclaration_351 {
   class=MaterialExpressionNamedRerouteDeclaration
   inputs [None {from=MaterialExpressionVectorParameter_53, output=Specular Exp}]
   properties {name="Asset Specular Exp (S)", nodeColor=<rgba 0 1 0.1094 1>}
  }
  MaterialExpressionNamedRerouteDeclaration_352 {
   class=MaterialExpressionNamedRerouteDeclaration
   inputs [None {from=MaterialExpressionVectorParameter_53, output=Specular Mult}]
   properties {name="Asset Specular Mult (S)", nodeColor=<rgba 1 0 0.2109 1>}
  }
  MaterialExpressionNamedRerouteDeclaration_377 {
   class=MaterialExpressionNamedRerouteDeclaration
   inputs [None {from=MaterialExpressionScalarParameter_97}]
   properties {name="Refracted Sky Desaturation (S)", nodeColor=<rgba 1 0.7734 0 1>}
  }
  MaterialExpressionNamedRerouteDeclaration_378 {
   class=MaterialExpressionNamedRerouteDeclaration
   inputs [None {from=MaterialExpressionScalarParameter_98}]
   properties {name="Diff Map Contrast (S)", nodeColor=<rgba 0.6641 1 0 1>}
  }
  MaterialExpressionNamedRerouteDeclaration_379 {
   class=MaterialExpressionNamedRerouteDeclaration
   inputs [None {from=MaterialExpressionVectorParameter_50, output=RGBA}]
   properties {name="Interior Noise Settings (V4)", nodeColor=<rgba 1 0.5859 0 1>}
  }
  MaterialExpressionNamedRerouteDeclaration_380 {
   class=MaterialExpressionNamedRerouteDeclaration
   inputs [None {from=MaterialExpressionVectorParameter_51, output=RGBA}]
   properties {name="Noise Depth Variation Settings (V4)", nodeColor=<rgba 0.2656 0 1 1>}
  }
  MaterialExpressionNamedRerouteDeclaration_381 {
   class=MaterialExpressionNamedRerouteDeclaration
   inputs [None {from=MaterialExpressionMultiply_112}]
   properties {name="UV Island - Mask (S)", nodeColor=<rgba 1 0 0.0938 1>}
  }
  MaterialExpressionNamedRerouteDeclaration_382 {
   class=MaterialExpressionNamedRerouteDeclaration
   inputs [None {from=MaterialExpressionAppendVector_17}]
   properties {name="UV Island - Min UV (V2)", nodeColor=<rgba 1 0 0.0938 1>}
  }
  MaterialExpressionNamedRerouteDeclaration_383 {
   class=MaterialExpressionNamedRerouteDeclaration
   inputs [None {from=MaterialExpressionAppendVector_18}]
   properties {name="UV Island - Max UV (V2)", nodeColor=<rgba 1 0 0.0938 1>}
  }
  MaterialExpressionNamedRerouteDeclaration_384 {
   class=MaterialExpressionNamedRerouteDeclaration
   inputs [None {from=MaterialExpressionSubtract_4}]
   properties {name="UV Island - Size (V2)", nodeColor=<rgba 1 0 0.0938 1>}
  }
  MaterialExpressionNamedRerouteDeclaration_385 {
   class=MaterialExpressionNamedRerouteDeclaration
   inputs [None {from=MaterialExpressionAdd_28}]
   properties {name="Emission Refracted UV [Raw] (V2)", nodeColor=<rgba 0 0.487 1 1>}
  }
  MaterialExpressionNamedRerouteDeclaration_386 {
   class=MaterialExpressionNamedRerouteDeclaration
   inputs [None {from=MaterialExpressionAdd_34}]
   properties {name="Diffuse Refracted UV [Raw] (V2)", nodeColor=<rgba 0 0.487 1 1>}
  }
  MaterialExpressionNamedRerouteDeclaration_387 {
   class=MaterialExpressionNamedRerouteDeclaration
   inputs [None {from=MaterialExpressionLinearInterpolate_44}]
   properties {name="Emission Refracted UV (V2)", nodeColor=<rgba 0 0.487 1 1>}
  }
  MaterialExpressionNamedRerouteDeclaration_388 {
   class=MaterialExpressionNamedRerouteDeclaration
   inputs [None {from=MaterialExpressionLinearInterpolate_45}]
   properties {name="Diffuse Refracted UV (V2)", nodeColor=<rgba 0 0.487 1 1>}
  }
  MaterialExpressionNamedRerouteDeclaration_389 {
   class=MaterialExpressionNamedRerouteDeclaration
   inputs [None {from=MaterialExpressionDDX_0}]
   properties {name="UV Asset DDX (V2)", nodeColor=<rgba 1 0 0.4219 1>}
  }
  MaterialExpressionNamedRerouteDeclaration_390 {
   class=MaterialExpressionNamedRerouteDeclaration
   inputs [None {from=MaterialExpressionDDY_0}]
   properties {name="UV Asset DDY (V2)", nodeColor=<rgba 1 0 0.4219 1>}
  }
  MaterialExpressionNamedRerouteDeclaration_391 {
   class=MaterialExpressionNamedRerouteDeclaration
   inputs [None {from=MaterialExpressionFunctionInput_0}]
   properties {name="Material In (MA)", nodeColor=<rgba 1 0.7969 0 1>}
  }
  MaterialExpressionNamedRerouteDeclaration_392 {
   class=MaterialExpressionNamedRerouteDeclaration
   inputs [None {from=MaterialExpressionGetMaterialAttributes_0, output=Normal}]
   properties {name="Surface Normal In TS (V3)", nodeColor=<rgba 0.4297 0 1 1>}
  }
  MaterialExpressionNamedRerouteDeclaration_393 {
   class=MaterialExpressionNamedRerouteDeclaration
   inputs [None {from=MaterialExpressionGetMaterialAttributes_0, output=Specular}]
   properties {name="Surface Specular In (S)", nodeColor=<rgba 1 0 0.5156 1>}
  }
  MaterialExpressionNamedRerouteDeclaration_394 {
   class=MaterialExpressionNamedRerouteDeclaration
   inputs [None {from=MaterialExpressionGetMaterialAttributes_0, output=Roughness}]
   properties {name="Surface Roughness In (S)", nodeColor=<rgba 1 0 0.5156 1>}
  }
  MaterialExpressionNamedRerouteDeclaration_396 {
   class=MaterialExpressionNamedRerouteDeclaration
   inputs [None {from=MaterialExpressionSaturate_40}]
   properties {name="Surface Opacity In (S)", nodeColor=<rgba 0.5234 0 1 1>}
  }
  MaterialExpressionNamedRerouteDeclaration_397 {
   class=MaterialExpressionNamedRerouteDeclaration
   inputs [None {from=MaterialExpressionGetMaterialAttributes_0, output=BaseColor}]
   properties {name="Surface Color In (V3)", nodeColor=<rgba 1 0.1172 0 1>}
  }
  MaterialExpressionNamedRerouteDeclaration_398 {
   class=MaterialExpressionNamedRerouteDeclaration
   inputs [None {from=MaterialExpressionTextureObjectParameter_7}]
   properties {name="Emissive Map (TO)", nodeColor=<rgba 1 0 0.5156 1>}
  }
  MaterialExpressionNamedRerouteDeclaration_399 {
   class=MaterialExpressionNamedRerouteDeclaration
   inputs [None {from=MaterialExpressionScalarParameter_106}]
   properties {name="Ice Mask (S)", nodeColor=<rgba 1 0.9375 0 1>}
  }
  MaterialExpressionNamedRerouteDeclaration_4 {
   class=MaterialExpressionNamedRerouteDeclaration
   inputs [None {from=MaterialExpressionVectorParameter_0, output=Bottom Ring}]
   properties {name="Detail UV Scale - Bottom Ring (S)", nodeColor=<rgba 0.4766 1 0 1>}
  }
  MaterialExpressionNamedRerouteDeclaration_400 {
   class=MaterialExpressionNamedRerouteDeclaration
   inputs [None {from=MaterialExpressionFunctionInput_1}]
   properties {name="Tech Art Mask In (V3)", nodeColor=<rgba 0.875 0 1 1>}
  }
  MaterialExpressionNamedRerouteDeclaration_401 {
   class=MaterialExpressionNamedRerouteDeclaration
   inputs [None {from=MaterialExpressionSaturate_39}]
   properties {name="Interface Opacity (S)", nodeColor=<rgba 0.7109 0 1 1>}
  }
  MaterialExpressionNamedRerouteDeclaration_402 {
   class=MaterialExpressionNamedRerouteDeclaration
   inputs [None {from=MaterialExpressionGetMaterialAttributes_0, output=EmissiveColor}]
   properties {name="Emissive Color In (V3)", nodeColor=<rgba 0 1 0.6953 1>}
  }
  MaterialExpressionNamedRerouteDeclaration_404 {
   class=MaterialExpressionNamedRerouteDeclaration
   inputs [None {from=MaterialExpressionFunctionInput_2}]
   properties {name="Additional Interior Emission (V3)", nodeColor=<rgba 1 0 0.3047 1>}
  }
  MaterialExpressionNamedRerouteDeclaration_6 {
   class=MaterialExpressionNamedRerouteDeclaration
   inputs [None {from=MaterialExpressionConvert_1}]
   properties {name="Rod Mask - Is Low (S)", nodeColor=<rgba 0.0312 1 0 1>}
  }
  MaterialExpressionNamedRerouteDeclaration_7 {
   class=MaterialExpressionNamedRerouteDeclaration
   inputs [None {from=MaterialExpressionConvert_1}]
   properties {name="Top Ring Mask - Is Low (S)", nodeColor=<rgba 0.0312 1 0 1>}
  }
  MaterialExpressionNamedRerouteDeclaration_8 {
   class=MaterialExpressionNamedRerouteDeclaration
   inputs [None {from=MaterialExpressionConvert_1}]
   properties {name="Middle RIng Mask - Is Low (S)", nodeColor=<rgba 0.0312 1 0 1>}
  }
  MaterialExpressionNamedRerouteDeclaration_9 {
   class=MaterialExpressionNamedRerouteDeclaration
   inputs [None {from=MaterialExpressionConvert_1}]
   properties {name="Bottom Ring Mask - Is Low (S)", nodeColor=<rgba 0.0312 1 0 1>}
  }
  MaterialExpressionNamedRerouteUsage_1 {class=MaterialExpressionNamedRerouteUsage}
  MaterialExpressionNamedRerouteUsage_10 {class=MaterialExpressionNamedRerouteUsage}
  MaterialExpressionNamedRerouteUsage_11 {class=MaterialExpressionNamedRerouteUsage}
  MaterialExpressionNamedRerouteUsage_12 {class=MaterialExpressionNamedRerouteUsage}
  MaterialExpressionNamedRerouteUsage_13 {class=MaterialExpressionNamedRerouteUsage}
  MaterialExpressionNamedRerouteUsage_19 {class=MaterialExpressionNamedRerouteUsage}
  MaterialExpressionNamedRerouteUsage_2 {class=MaterialExpressionNamedRerouteUsage}
  MaterialExpressionNamedRerouteUsage_20 {class=MaterialExpressionNamedRerouteUsage}
  MaterialExpressionNamedRerouteUsage_21 {class=MaterialExpressionNamedRerouteUsage}
  MaterialExpressionNamedRerouteUsage_22 {class=MaterialExpressionNamedRerouteUsage}
  MaterialExpressionNamedRerouteUsage_23 {class=MaterialExpressionNamedRerouteUsage}
  MaterialExpressionNamedRerouteUsage_24 {class=MaterialExpressionNamedRerouteUsage}
  MaterialExpressionNamedRerouteUsage_25 {class=MaterialExpressionNamedRerouteUsage}
  MaterialExpressionNamedRerouteUsage_26 {class=MaterialExpressionNamedRerouteUsage}
  MaterialExpressionNamedRerouteUsage_27 {class=MaterialExpressionNamedRerouteUsage}
  MaterialExpressionNamedRerouteUsage_3 {class=MaterialExpressionNamedRerouteUsage}
  MaterialExpressionNamedRerouteUsage_373 {class=MaterialExpressionNamedRerouteUsage}
  MaterialExpressionNamedRerouteUsage_374 {class=MaterialExpressionNamedRerouteUsage}
  MaterialExpressionNamedRerouteUsage_375 {class=MaterialExpressionNamedRerouteUsage}
  MaterialExpressionNamedRerouteUsage_376 {class=MaterialExpressionNamedRerouteUsage}
  MaterialExpressionNamedRerouteUsage_377 {class=MaterialExpressionNamedRerouteUsage}
  MaterialExpressionNamedRerouteUsage_378 {class=MaterialExpressionNamedRerouteUsage}
  MaterialExpressionNamedRerouteUsage_379 {class=MaterialExpressionNamedRerouteUsage}
  MaterialExpressionNamedRerouteUsage_380 {class=MaterialExpressionNamedRerouteUsage}
  MaterialExpressionNamedRerouteUsage_381 {class=MaterialExpressionNamedRerouteUsage}
  MaterialExpressionNamedRerouteUsage_384 {class=MaterialExpressionNamedRerouteUsage}
  MaterialExpressionNamedRerouteUsage_385 {class=MaterialExpressionNamedRerouteUsage}
  MaterialExpressionNamedRerouteUsage_386 {class=MaterialExpressionNamedRerouteUsage}
  MaterialExpressionNamedRerouteUsage_387 {class=MaterialExpressionNamedRerouteUsage}
  MaterialExpressionNamedRerouteUsage_388 {class=MaterialExpressionNamedRerouteUsage}
  MaterialExpressionNamedRerouteUsage_390 {class=MaterialExpressionNamedRerouteUsage}
  MaterialExpressionNamedRerouteUsage_391 {class=MaterialExpressionNamedRerouteUsage}
  MaterialExpressionNamedRerouteUsage_392 {class=MaterialExpressionNamedRerouteUsage}
  MaterialExpressionNamedRerouteUsage_393 {class=MaterialExpressionNamedRerouteUsage}
  MaterialExpressionNamedRerouteUsage_394 {class=MaterialExpressionNamedRerouteUsage}
  MaterialExpressionNamedRerouteUsage_395 {class=MaterialExpressionNamedRerouteUsage}
  MaterialExpressionNamedRerouteUsage_396 {class=MaterialExpressionNamedRerouteUsage}
  MaterialExpressionNamedRerouteUsage_397 {class=MaterialExpressionNamedRerouteUsage}
  MaterialExpressionNamedRerouteUsage_398 {class=MaterialExpressionNamedRerouteUsage}
  MaterialExpressionNamedRerouteUsage_399 {class=MaterialExpressionNamedRerouteUsage}
  MaterialExpressionNamedRerouteUsage_4 {class=MaterialExpressionNamedRerouteUsage}
  MaterialExpressionNamedRerouteUsage_400 {class=MaterialExpressionNamedRerouteUsage}
  MaterialExpressionNamedRerouteUsage_401 {class=MaterialExpressionNamedRerouteUsage}
  MaterialExpressionNamedRerouteUsage_402 {class=MaterialExpressionNamedRerouteUsage}
  MaterialExpressionNamedRerouteUsage_403 {class=MaterialExpressionNamedRerouteUsage}
  MaterialExpressionNamedRerouteUsage_404 {class=MaterialExpressionNamedRerouteUsage}
  MaterialExpressionNamedRerouteUsage_405 {class=MaterialExpressionNamedRerouteUsage}
  MaterialExpressionNamedRerouteUsage_406 {class=MaterialExpressionNamedRerouteUsage}
  MaterialExpressionNamedRerouteUsage_407 {class=MaterialExpressionNamedRerouteUsage}
  MaterialExpressionNamedRerouteUsage_408 {class=MaterialExpressionNamedRerouteUsage}
  MaterialExpressionNamedRerouteUsage_409 {class=MaterialExpressionNamedRerouteUsage}
  MaterialExpressionNamedRerouteUsage_410 {class=MaterialExpressionNamedRerouteUsage}
  MaterialExpressionNamedRerouteUsage_411 {class=MaterialExpressionNamedRerouteUsage}
  MaterialExpressionNamedRerouteUsage_412 {class=MaterialExpressionNamedRerouteUsage}
  MaterialExpressionNamedRerouteUsage_413 {class=MaterialExpressionNamedRerouteUsage}
  MaterialExpressionNamedRerouteUsage_414 {class=MaterialExpressionNamedRerouteUsage}
  MaterialExpressionNamedRerouteUsage_415 {class=MaterialExpressionNamedRerouteUsage}
  MaterialExpressionNamedRerouteUsage_416 {class=MaterialExpressionNamedRerouteUsage}
  MaterialExpressionNamedRerouteUsage_417 {class=MaterialExpressionNamedRerouteUsage}
  MaterialExpressionNamedRerouteUsage_418 {class=MaterialExpressionNamedRerouteUsage}
  MaterialExpressionNamedRerouteUsage_421 {class=MaterialExpressionNamedRerouteUsage}
  MaterialExpressionNamedRerouteUsage_422 {class=MaterialExpressionNamedRerouteUsage}
  MaterialExpressionNamedRerouteUsage_424 {class=MaterialExpressionNamedRerouteUsage}
  MaterialExpressionNamedRerouteUsage_425 {class=MaterialExpressionNamedRerouteUsage}
  MaterialExpressionNamedRerouteUsage_426 {class=MaterialExpressionNamedRerouteUsage}
  MaterialExpressionNamedRerouteUsage_427 {class=MaterialExpressionNamedRerouteUsage}
  MaterialExpressionNamedRerouteUsage_428 {class=MaterialExpressionNamedRerouteUsage}
  MaterialExpressionNamedRerouteUsage_430 {class=MaterialExpressionNamedRerouteUsage}
  MaterialExpressionNamedRerouteUsage_431 {class=MaterialExpressionNamedRerouteUsage}
  MaterialExpressionNamedRerouteUsage_432 {class=MaterialExpressionNamedRerouteUsage}
  MaterialExpressionNamedRerouteUsage_433 {class=MaterialExpressionNamedRerouteUsage}
  MaterialExpressionNamedRerouteUsage_434 {class=MaterialExpressionNamedRerouteUsage}
  MaterialExpressionNamedRerouteUsage_435 {class=MaterialExpressionNamedRerouteUsage}
  MaterialExpressionNamedRerouteUsage_436 {class=MaterialExpressionNamedRerouteUsage}
  MaterialExpressionNamedRerouteUsage_437 {class=MaterialExpressionNamedRerouteUsage}
  MaterialExpressionNamedRerouteUsage_438 {class=MaterialExpressionNamedRerouteUsage}
  MaterialExpressionNamedRerouteUsage_439 {class=MaterialExpressionNamedRerouteUsage}
  MaterialExpressionNamedRerouteUsage_440 {class=MaterialExpressionNamedRerouteUsage}
  MaterialExpressionNamedRerouteUsage_441 {class=MaterialExpressionNamedRerouteUsage}
  MaterialExpressionNamedRerouteUsage_442 {class=MaterialExpressionNamedRerouteUsage}
  MaterialExpressionNamedRerouteUsage_443 {class=MaterialExpressionNamedRerouteUsage}
  MaterialExpressionNamedRerouteUsage_444 {class=MaterialExpressionNamedRerouteUsage}
  MaterialExpressionNamedRerouteUsage_445 {class=MaterialExpressionNamedRerouteUsage}
  MaterialExpressionNamedRerouteUsage_446 {class=MaterialExpressionNamedRerouteUsage}
  MaterialExpressionNamedRerouteUsage_447 {class=MaterialExpressionNamedRerouteUsage}
  MaterialExpressionNamedRerouteUsage_449 {class=MaterialExpressionNamedRerouteUsage}
  MaterialExpressionNamedRerouteUsage_450 {class=MaterialExpressionNamedRerouteUsage}
  MaterialExpressionNamedRerouteUsage_451 {class=MaterialExpressionNamedRerouteUsage}
  MaterialExpressionNamedRerouteUsage_452 {class=MaterialExpressionNamedRerouteUsage}
  MaterialExpressionNamedRerouteUsage_453 {class=MaterialExpressionNamedRerouteUsage}
  MaterialExpressionNamedRerouteUsage_455 {class=MaterialExpressionNamedRerouteUsage}
  MaterialExpressionNamedRerouteUsage_456 {class=MaterialExpressionNamedRerouteUsage}
  MaterialExpressionNamedRerouteUsage_457 {class=MaterialExpressionNamedRerouteUsage}
  MaterialExpressionNamedRerouteUsage_459 {class=MaterialExpressionNamedRerouteUsage}
  MaterialExpressionNamedRerouteUsage_476 {class=MaterialExpressionNamedRerouteUsage}
  MaterialExpressionNamedRerouteUsage_477 {class=MaterialExpressionNamedRerouteUsage}
  MaterialExpressionNamedRerouteUsage_478 {class=MaterialExpressionNamedRerouteUsage}
  MaterialExpressionNamedRerouteUsage_479 {class=MaterialExpressionNamedRerouteUsage}
  MaterialExpressionNamedRerouteUsage_480 {class=MaterialExpressionNamedRerouteUsage}
  MaterialExpressionNamedRerouteUsage_481 {class=MaterialExpressionNamedRerouteUsage}
  MaterialExpressionNamedRerouteUsage_482 {class=MaterialExpressionNamedRerouteUsage}
  MaterialExpressionNamedRerouteUsage_483 {class=MaterialExpressionNamedRerouteUsage}
  MaterialExpressionNamedRerouteUsage_486 {class=MaterialExpressionNamedRerouteUsage}
  MaterialExpressionNamedRerouteUsage_487 {class=MaterialExpressionNamedRerouteUsage}
  MaterialExpressionNamedRerouteUsage_488 {class=MaterialExpressionNamedRerouteUsage}
  MaterialExpressionNamedRerouteUsage_489 {class=MaterialExpressionNamedRerouteUsage}
  MaterialExpressionNamedRerouteUsage_492 {class=MaterialExpressionNamedRerouteUsage}
  MaterialExpressionNamedRerouteUsage_493 {class=MaterialExpressionNamedRerouteUsage}
  MaterialExpressionNamedRerouteUsage_494 {class=MaterialExpressionNamedRerouteUsage}
  MaterialExpressionNamedRerouteUsage_495 {class=MaterialExpressionNamedRerouteUsage}
  MaterialExpressionNamedRerouteUsage_496 {class=MaterialExpressionNamedRerouteUsage}
  MaterialExpressionNamedRerouteUsage_497 {class=MaterialExpressionNamedRerouteUsage}
  MaterialExpressionNamedRerouteUsage_5 {class=MaterialExpressionNamedRerouteUsage}
  MaterialExpressionNamedRerouteUsage_503 {class=MaterialExpressionNamedRerouteUsage}
  MaterialExpressionNamedRerouteUsage_538 {class=MaterialExpressionNamedRerouteUsage}
  MaterialExpressionNamedRerouteUsage_539 {class=MaterialExpressionNamedRerouteUsage}
  MaterialExpressionNamedRerouteUsage_540 {class=MaterialExpressionNamedRerouteUsage}
  MaterialExpressionNamedRerouteUsage_541 {class=MaterialExpressionNamedRerouteUsage}
  MaterialExpressionNamedRerouteUsage_543 {class=MaterialExpressionNamedRerouteUsage}
  MaterialExpressionNamedRerouteUsage_544 {class=MaterialExpressionNamedRerouteUsage}
  MaterialExpressionNamedRerouteUsage_545 {class=MaterialExpressionNamedRerouteUsage}
  MaterialExpressionNamedRerouteUsage_546 {class=MaterialExpressionNamedRerouteUsage}
  MaterialExpressionNamedRerouteUsage_547 {class=MaterialExpressionNamedRerouteUsage}
  MaterialExpressionNamedRerouteUsage_548 {class=MaterialExpressionNamedRerouteUsage}
  MaterialExpressionNamedRerouteUsage_549 {class=MaterialExpressionNamedRerouteUsage}
  MaterialExpressionNamedRerouteUsage_550 {class=MaterialExpressionNamedRerouteUsage}
  MaterialExpressionNamedRerouteUsage_551 {class=MaterialExpressionNamedRerouteUsage}
  MaterialExpressionNamedRerouteUsage_552 {class=MaterialExpressionNamedRerouteUsage}
  MaterialExpressionNamedRerouteUsage_553 {class=MaterialExpressionNamedRerouteUsage}
  MaterialExpressionNamedRerouteUsage_554 {class=MaterialExpressionNamedRerouteUsage}
  MaterialExpressionNamedRerouteUsage_555 {class=MaterialExpressionNamedRerouteUsage}
  MaterialExpressionNamedRerouteUsage_556 {class=MaterialExpressionNamedRerouteUsage}
  MaterialExpressionNamedRerouteUsage_557 {class=MaterialExpressionNamedRerouteUsage}
  MaterialExpressionNamedRerouteUsage_558 {class=MaterialExpressionNamedRerouteUsage}
  MaterialExpressionNamedRerouteUsage_559 {class=MaterialExpressionNamedRerouteUsage}
  MaterialExpressionNamedRerouteUsage_560 {class=MaterialExpressionNamedRerouteUsage}
  MaterialExpressionNamedRerouteUsage_561 {class=MaterialExpressionNamedRerouteUsage}
  MaterialExpressionNamedRerouteUsage_562 {class=MaterialExpressionNamedRerouteUsage}
  MaterialExpressionNamedRerouteUsage_564 {class=MaterialExpressionNamedRerouteUsage}
  MaterialExpressionNamedRerouteUsage_565 {class=MaterialExpressionNamedRerouteUsage}
  MaterialExpressionNamedRerouteUsage_566 {class=MaterialExpressionNamedRerouteUsage}
  MaterialExpressionNamedRerouteUsage_567 {class=MaterialExpressionNamedRerouteUsage}
  MaterialExpressionNamedRerouteUsage_568 {class=MaterialExpressionNamedRerouteUsage}
  MaterialExpressionNamedRerouteUsage_6 {class=MaterialExpressionNamedRerouteUsage}
  MaterialExpressionNamedRerouteUsage_7 {class=MaterialExpressionNamedRerouteUsage}
  MaterialExpressionNamedRerouteUsage_8 {class=MaterialExpressionNamedRerouteUsage}
  MaterialExpressionNamedRerouteUsage_9 {class=MaterialExpressionNamedRerouteUsage}
  MaterialExpressionNormalize_2 {class=MaterialExpressionNormalize, inputs [VectorInput {from=MaterialExpressionTransform_2, output=""}]}
  MaterialExpressionOneMinus_0 {class=MaterialExpressionOneMinus, inputs [None {from=MaterialExpressionStep_0}]}
  MaterialExpressionOneMinus_1 {class=MaterialExpressionOneMinus, inputs [None {from=MaterialExpressionStep_8}]}
  MaterialExpressionOneMinus_11 {class=MaterialExpressionOneMinus, inputs [None {from=MaterialExpressionStep_3}]}
  MaterialExpressionOneMinus_2 {class=MaterialExpressionOneMinus, inputs [None {from=MaterialExpressionStep_10}]}
  MaterialExpressionOneMinus_3 {class=MaterialExpressionOneMinus, inputs [None {from=MaterialExpressionStep_12}]}
  MaterialExpressionOneMinus_8 {class=MaterialExpressionOneMinus, inputs [None {from=MaterialExpressionLinearInterpolate_32}]}
  MaterialExpressionOneMinus_9 {class=MaterialExpressionOneMinus, inputs [None {from=MaterialExpressionNamedRerouteUsage_430}]}
  MaterialExpressionPower_28 {
   class=MaterialExpressionPower
   inputs [
    Base {from=MaterialExpressionNamedRerouteDeclaration_287}
    Exp {from=MaterialExpressionNamedRerouteUsage_411}
   ]
  }
  MaterialExpressionPower_29 {
   class=MaterialExpressionPower
   inputs [
    Base {from=MaterialExpressionTextureSample_16, output=RGB}
    Exp {from=MaterialExpressionNamedRerouteUsage_444}
   ]
  }
  MaterialExpressionPower_30 {
   class=MaterialExpressionPower
   inputs [
    Base {from=MaterialExpressionOneMinus_8}
    Exp {from=MaterialExpressionNamedRerouteUsage_427}
   ]
  }
  MaterialExpressionPower_33 {
   class=MaterialExpressionPower
   inputs [
    Base {from=MaterialExpressionNamedRerouteUsage_425}
    Exp {from=MaterialExpressionNamedRerouteUsage_431}
   ]
  }
  MaterialExpressionPower_34 {
   class=MaterialExpressionPower
   inputs [
    Base {from=MaterialExpressionOneMinus_9}
    Exp {from=MaterialExpressionNamedRerouteUsage_437}
   ]
  }
  MaterialExpressionPower_35 {
   class=MaterialExpressionPower
   inputs [
    Base {from=MaterialExpressionTextureSample_13, output=R}
    Exp {from=MaterialExpressionNamedRerouteUsage_446}
   ]
  }
  MaterialExpressionPower_39 {
   class=MaterialExpressionPower
   inputs [
    Base {from=MaterialExpressionNamedRerouteUsage_487}
    Exp {from=MaterialExpressionNamedRerouteUsage_496}
   ]
  }
  MaterialExpressionPower_40 {
   class=MaterialExpressionPower
   inputs [
    Base {from=MaterialExpressionNamedRerouteUsage_488}
    Exp {from=MaterialExpressionNamedRerouteUsage_493}
   ]
  }
  MaterialExpressionReroute_24 {class=MaterialExpressionReroute, inputs [None {from=MaterialExpressionTextureObject_4}]}
  MaterialExpressionReroute_25 {class=MaterialExpressionReroute, inputs [None {from=MaterialExpressionScalarParameter_102}]}
  MaterialExpressionReroute_26 {class=MaterialExpressionReroute, inputs [None {from=MaterialExpressionScalarParameter_101}]}
  MaterialExpressionReroute_27 {class=MaterialExpressionReroute, inputs [None {from=MaterialExpressionScalarParameter_99}]}
  MaterialExpressionReroute_28 {class=MaterialExpressionReroute, inputs [None {from=MaterialExpressionNamedRerouteUsage_540}]}
  MaterialExpressionReroute_29 {class=MaterialExpressionReroute, inputs [None {from=MaterialExpressionNamedRerouteUsage_541}]}
  MaterialExpressionReroute_30 {class=MaterialExpressionReroute, inputs [None {from=MaterialExpressionNamedRerouteDeclaration_294}]}
  MaterialExpressionReroute_31 {class=MaterialExpressionReroute, inputs [None {from=MaterialExpressionNamedRerouteUsage_543}]}
  MaterialExpressionReroute_32 {class=MaterialExpressionReroute, inputs [None {from=MaterialExpressionScalarParameter_100}]}
  MaterialExpressionReroute_33 {class=MaterialExpressionReroute, inputs [None {from=MaterialExpressionScalarParameter_103}]}
  MaterialExpressionReroute_34 {class=MaterialExpressionReroute, inputs [None {from=MaterialExpressionScalarParameter_104}]}
  MaterialExpressionReroute_35 {class=MaterialExpressionReroute, inputs [None {from=MaterialExpressionScalarParameter_105}]}
  MaterialExpressionRgbToHsv_2 {class=MaterialExpressionRgbToHsv, inputs [None {from=MaterialExpressionPower_33, output=""}], properties {}}
  MaterialExpressionRound_0 {class=MaterialExpressionRound, inputs [None {from=MaterialExpressionSaturate_0, output=""}], properties {}}
  MaterialExpressionSaturate_0 {class=MaterialExpressionSaturate, inputs [None {from=MaterialExpressionVectorParameter_3, output=RGBA}]}
  MaterialExpressionSaturate_27 {class=MaterialExpressionSaturate, inputs [None {from=MaterialExpressionDotProduct_13}]}
  MaterialExpressionSaturate_28 {class=MaterialExpressionSaturate, inputs [None {from=MaterialExpressionMultiply_85}]}
  MaterialExpressionSaturate_29 {class=MaterialExpressionSaturate, inputs [None {from=MaterialExpressionDotProduct_14}]}
  MaterialExpressionSaturate_30 {class=MaterialExpressionSaturate, inputs [None {from=MaterialExpressionMultiply_93}]}
  MaterialExpressionSaturate_36 {class=MaterialExpressionSaturate, inputs [None {from=MaterialExpressionMultiply_101}]}
  MaterialExpressionSaturate_37 {class=MaterialExpressionSaturate, inputs [None {from=MaterialExpressionMultiply_102}]}
  MaterialExpressionSaturate_38 {class=MaterialExpressionSaturate, inputs [None {from=MaterialExpressionOneMinus_8}]}
  MaterialExpressionSaturate_39 {class=MaterialExpressionSaturate, inputs [None {from=MaterialExpressionScalarParameter_107}]}
  MaterialExpressionSaturate_40 {class=MaterialExpressionSaturate, inputs [None {from=MaterialExpressionScalarParameter_108}]}
  MaterialExpressionScalarParameter_0 {
   class=MaterialExpressionScalarParameter
   properties {defaultValue=0.1458, group=Ice | 02_Surface, parameterName=Surface - Detail Tiles Base Scale, sortPriority=120}
  }
  MaterialExpressionScalarParameter_100 {
   class=MaterialExpressionScalarParameter
   properties {defaultValue=100, group=Ice | 07_Depth Pattern, parameterName=Triplanar Parallax Depth Norm, sortPriority=40}
  }
  MaterialExpressionScalarParameter_101 {
   class=MaterialExpressionScalarParameter
   properties {defaultValue=0.7, group=Ice | 07_Depth Pattern, parameterName=Triplanar V Tiling Ratio, sortPriority=20}
  }
  MaterialExpressionScalarParameter_102 {
   class=MaterialExpressionScalarParameter
   properties {defaultValue=10, group=Ice | 07_Depth Pattern, parameterName=Triplanar Tiling, sortPriority=10}
  }
  MaterialExpressionScalarParameter_103 {
   class=MaterialExpressionScalarParameter
   properties {defaultValue=1, group=Ice | 07_Depth Pattern, parameterName=Triplanar Parallax Projection Min Cos, sortPriority=50}
  }
  MaterialExpressionScalarParameter_104 {
   class=MaterialExpressionScalarParameter
   properties {defaultValue=0.5, group=Ice | 07_Depth Pattern, parameterName=Triplanar Parallax Grazing Angle Reduction, sortPriority=60}
  }
  MaterialExpressionScalarParameter_105 {
   class=MaterialExpressionScalarParameter
   properties {defaultValue=0.25, group=Ice | 07_Depth Pattern, parameterName=Triplanar Parallax Reduction Min Cos, sortPriority=70}
  }
  MaterialExpressionScalarParameter_106 {
   class=MaterialExpressionScalarParameter
   properties {defaultValue=0.9, group=Ice | 01_Globals, parameterName=Ice Mask, sliderMax=1, sortPriority=20}
  }
  MaterialExpressionScalarParameter_107 {
   class=MaterialExpressionScalarParameter
   properties {group=Ice | 01_Globals, parameterName=Ice Interface Opacity, sliderMax=1, sortPriority=130}
  }
  MaterialExpressionScalarParameter_108 {
   class=MaterialExpressionScalarParameter
   properties {group=Ice | 01_Globals, parameterName=Ice Surface Opacity, sliderMax=1, sortPriority=120}
  }
  MaterialExpressionScalarParameter_65 {
   class=MaterialExpressionScalarParameter
   properties {defaultValue=1, group=Ice | 04_Volume, parameterName=Volume - Thickness Gradient Exp, sliderMax=10, sliderMin=0.01, sortPriority=10}
  }
  MaterialExpressionScalarParameter_66 {
   class=MaterialExpressionScalarParameter
   properties {defaultValue=10, group=Ice | 04_Volume, parameterName=Volume - Max Thickness, sortPriority=20}
  }
  MaterialExpressionScalarParameter_67 {
   class=MaterialExpressionScalarParameter
   properties {defaultValue=1.5, group=Ice | 03_Refraction, parameterName=Refraction - IOR, sliderMax=5, sliderMin=1, sortPriority=30}
  }
  MaterialExpressionScalarParameter_68 {
   class=MaterialExpressionScalarParameter
   properties {defaultValue=0.2, group=Ice | 03_Refraction, parameterName=Refraction - Roughness, sliderMax=1, sortPriority=40}
  }
  MaterialExpressionScalarParameter_69 {
   class=MaterialExpressionScalarParameter
   properties {defaultValue=1, group=Ice | 06_Emission, parameterName=Emission - Refracted Emission Intensity, sortPriority=30}
  }
  MaterialExpressionScalarParameter_70 {
   class=MaterialExpressionScalarParameter
   properties {defaultValue=1, group=Ice | 06_Emission, parameterName=Emission - Refracted Emission Exp, sortPriority=20}
  }
  MaterialExpressionScalarParameter_71 {
   class=MaterialExpressionScalarParameter
   properties {defaultValue=0.12, group=Ice | 06_Emission, parameterName=Emission - Depth, sortPriority=50}
  }
  MaterialExpressionScalarParameter_72 {
   class=MaterialExpressionScalarParameter
   properties {defaultValue=0.5, group=Ice | 05 _Absorption, parameterName=Absorption - Depth, sortPriority=30}
  }
  MaterialExpressionScalarParameter_74 {
   class=MaterialExpressionScalarParameter
   properties {defaultValue=1, group=Ice | 05 _Absorption, parameterName=Absorption - Depth Thickness Attenuation, sliderMax=1, sortPriority=40}
  }
  MaterialExpressionScalarParameter_75 {
   class=MaterialExpressionScalarParameter
   properties {defaultValue=1, group=Ice | 06_Emission, parameterName=Emission - Variation Noise Speed, sortPriority=90}
  }
  MaterialExpressionScalarParameter_79 {
   class=MaterialExpressionScalarParameter
   properties {defaultValue=1, group=Ice | 03_Refraction, parameterName=Refraction - Refracted Light Intensity, sliderMax=1, sortPriority=60}
  }
  MaterialExpressionScalarParameter_80 {class=MaterialExpressionScalarParameter, properties {defaultValue=1, group=Ice | 04_Volume, parameterName=Volume - Density, sortPriority=30}}
  MaterialExpressionScalarParameter_81 {
   class=MaterialExpressionScalarParameter
   properties {group=Ice | 04_Volume, parameterName=Volume - Scattering Phase, sliderMax=1, sliderMin=-1, sortPriority=60}
  }
  MaterialExpressionScalarParameter_82 {
   class=MaterialExpressionScalarParameter
   properties {defaultValue=1, group=Ice | 03_Refraction, parameterName=Refraction - Transmission Attenuation, sortPriority=50}
  }
  MaterialExpressionScalarParameter_97 {
   class=MaterialExpressionScalarParameter
   properties {group=Ice | 03_Refraction, parameterName=Refraction - Refracted Sky Desaturation, sliderMax=1, sortPriority=70}
  }
  MaterialExpressionScalarParameter_98 {
   class=MaterialExpressionScalarParameter
   properties {group=Ice | 05 _Absorption, parameterName=Absorption - Diffise Map Contrast, sortPriority=5}
  }
  MaterialExpressionScalarParameter_99 {
   class=MaterialExpressionScalarParameter
   properties {defaultValue=1, group=Ice | 07_Depth Pattern, parameterName=Triplanar Blend Contrast, sortPriority=30}
  }
  MaterialExpressionSetMaterialAttributes_4 {
   class=MaterialExpressionSetMaterialAttributes
   inputs [
    MaterialAttributes {from=MaterialExpressionNamedRerouteUsage_400, output=""}
    BaseColor {from=MaterialExpressionNamedRerouteUsage_401, output=""}
    EmissiveColor {from=MaterialExpressionNamedRerouteUsage_402, output=""}
    Metallic {from=MaterialExpressionConstant_33, output=""}
    Specular {from=MaterialExpressionNamedRerouteUsage_403, output=""}
    Roughness {from=MaterialExpressionNamedRerouteUsage_404, output=""}
    Normal {from=MaterialExpressionNamedRerouteUsage_405, output=""}
   ]
   properties {attributeSetTypes [
     #1
     #5
     #6
     #2
     #3
     #4
    ]}
  }
  MaterialExpressionSetMaterialAttributes_5 {
   class=MaterialExpressionSetMaterialAttributes
   inputs [
    MaterialAttributes {from=MaterialExpressionBlendMaterialAttributes_2, output=""}
    Metallic {from=MaterialExpressionConstant_41, output=""}
   ]
   properties {attributeSetTypes [#6]}
  }
  MaterialExpressionStaticBool_2 {class=MaterialExpressionStaticBool, properties {value=true}}
  MaterialExpressionStaticSwitchParameter_2 {
   class=MaterialExpressionStaticSwitchParameter
   inputs [
    True {from=MaterialExpressionSetMaterialAttributes_5, output=""}
    False {from=MaterialExpressionNamedRerouteUsage_385, output=""}
   ]
   properties {group=04_Tech Art, parameterName=UseIce}
  }
  MaterialExpressionStep_0 {
   class=MaterialExpressionStep
   inputs [
    Y {from="", output=""}
    X {from=MaterialExpressionNamedRerouteUsage_19, output=""}
   ]
   properties {constX=0.75, constY=0.75}
  }
  MaterialExpressionStep_1 {
   class=MaterialExpressionStep
   inputs [
    Y {from="", output=""}
    X {from=MaterialExpressionNamedRerouteUsage_19, output=""}
   ]
   properties {constX=0.75, constY=0.25}
  }
  MaterialExpressionStep_10 {
   class=MaterialExpressionStep
   inputs [
    Y {from="", output=""}
    X {from=MaterialExpressionNamedRerouteUsage_22, output=""}
   ]
   properties {constX=0.75, constY=0.75}
  }
  MaterialExpressionStep_11 {
   class=MaterialExpressionStep
   inputs [
    Y {from="", output=""}
    X {from=MaterialExpressionNamedRerouteUsage_22, output=""}
   ]
   properties {constX=0.75, constY=0.25}
  }
  MaterialExpressionStep_12 {
   class=MaterialExpressionStep
   inputs [
    Y {from="", output=""}
    X {from=MaterialExpressionNamedRerouteUsage_26, output=""}
   ]
   properties {constX=0.75, constY=0.75}
  }
  MaterialExpressionStep_13 {
   class=MaterialExpressionStep
   inputs [
    Y {from="", output=""}
    X {from=MaterialExpressionNamedRerouteUsage_26, output=""}
   ]
   properties {constX=0.75, constY=0.25}
  }
  MaterialExpressionStep_2 {
   class=MaterialExpressionStep
   inputs [
    Y {from=MaterialExpressionNamedRerouteUsage_545, output=""}
    X {from=MaterialExpressionNamedRerouteUsage_552, output=""}
   ]
   properties {}
  }
  MaterialExpressionStep_3 {
   class=MaterialExpressionStep
   inputs [
    Y {from=MaterialExpressionNamedRerouteUsage_544, output=""}
    X {from=MaterialExpressionNamedRerouteUsage_552, output=""}
   ]
   properties {}
  }
  MaterialExpressionStep_8 {
   class=MaterialExpressionStep
   inputs [
    Y {from="", output=""}
    X {from=MaterialExpressionNamedRerouteUsage_20, output=""}
   ]
   properties {constX=0.75, constY=0.75}
  }
  MaterialExpressionStep_9 {
   class=MaterialExpressionStep
   inputs [
    Y {from="", output=""}
    X {from=MaterialExpressionNamedRerouteUsage_20, output=""}
   ]
   properties {constX=0.75, constY=0.25}
  }
  MaterialExpressionSubtract_3 {
   class=MaterialExpressionSubtract
   inputs [
    A {from=MaterialExpressionNamedRerouteUsage_550}
    B {from=MaterialExpressionNamedRerouteUsage_546}
   ]
  }
  MaterialExpressionSubtract_4 {
   class=MaterialExpressionSubtract
   inputs [
    A {from=MaterialExpressionNamedRerouteDeclaration_383}
    B {from=MaterialExpressionNamedRerouteDeclaration_382}
   ]
  }
  MaterialExpressionSubtract_5 {
   class=MaterialExpressionSubtract
   inputs [
    A {from=MaterialExpressionNamedRerouteUsage_555}
    B {from=MaterialExpressionNamedRerouteUsage_556}
   ]
  }
  MaterialExpressionSwitch_0 {
   class=MaterialExpressionSwitch
   inputs [
    SwitchValue {from=MaterialExpressionNamedRerouteUsage_10, output=""}
    Default {from="", output=""}
    High {from=MaterialExpressionStep_0, output=""}
    Low {from=MaterialExpressionMultiply_0, output=""}
   ]
   properties {inputs [
     {
      inputName=High
      input {
       expression=&self:MaterialExpressionStep_0
       inputName=None
       mask=0
       maskA=0
       maskB=0
       maskG=0
       maskR=0
       outputIndex=0
      }
     }
     {
      inputName=Low
      input {
       expression=&self:MaterialExpressionMultiply_0
       inputName=None
       mask=0
       maskA=0
       maskB=0
       maskG=0
       maskR=0
       outputIndex=0
      }
     }
    ]}
  }
  MaterialExpressionSwitch_1 {
   class=MaterialExpressionSwitch
   inputs [
    SwitchValue {from=MaterialExpressionNamedRerouteUsage_11, output=""}
    Default {from="", output=""}
    High {from=MaterialExpressionStep_8, output=""}
    Low {from=MaterialExpressionMultiply_1, output=""}
   ]
   properties {inputs [
     {
      inputName=High
      input {
       expression=&self:MaterialExpressionStep_8
       inputName=None
       mask=0
       maskA=0
       maskB=0
       maskG=0
       maskR=0
       outputIndex=0
      }
     }
     {
      inputName=Low
      input {
       expression=&self:MaterialExpressionMultiply_1
       inputName=None
       mask=0
       maskA=0
       maskB=0
       maskG=0
       maskR=0
       outputIndex=0
      }
     }
    ]}
  }
  MaterialExpressionSwitch_2 {
   class=MaterialExpressionSwitch
   inputs [
    SwitchValue {from=MaterialExpressionNamedRerouteUsage_12, output=""}
    Default {from="", output=""}
    High {from=MaterialExpressionStep_10, output=""}
    Low {from=MaterialExpressionMultiply_2, output=""}
   ]
   properties {inputs [
     {
      inputName=High
      input {
       expression=&self:MaterialExpressionStep_10
       inputName=None
       mask=0
       maskA=0
       maskB=0
       maskG=0
       maskR=0
       outputIndex=0
      }
     }
     {
      inputName=Low
      input {
       expression=&self:MaterialExpressionMultiply_2
       inputName=None
       mask=0
       maskA=0
       maskB=0
       maskG=0
       maskR=0
       outputIndex=0
      }
     }
    ]}
  }
  MaterialExpressionSwitch_3 {
   class=MaterialExpressionSwitch
   inputs [
    SwitchValue {from=MaterialExpressionNamedRerouteUsage_13, output=""}
    Default {from="", output=""}
    High {from=MaterialExpressionStep_12, output=""}
    Low {from=MaterialExpressionMultiply_3, output=""}
   ]
   properties {inputs [
     {
      inputName=High
      input {
       expression=&self:MaterialExpressionStep_12
       inputName=None
       mask=0
       maskA=0
       maskB=0
       maskG=0
       maskR=0
       outputIndex=0
      }
     }
     {
      inputName=Low
      input {
       expression=&self:MaterialExpressionMultiply_3
       inputName=None
       mask=0
       maskA=0
       maskB=0
       maskG=0
       maskR=0
       outputIndex=0
      }
     }
    ]}
  }
  MaterialExpressionTextureCoordinate_2 {class=MaterialExpressionTextureCoordinate}
  MaterialExpressionTextureObjectParameter_5 {
   class=MaterialExpressionTextureObjectParameter
   inputs [
    UVs {from="", output=""}
    Apply View MipBias {from="", output=""}
   ]
   properties {
    group=Ice | 02_Surface
    parameterName=Surface - Detail Roughness Map
    samplerType=SAMPLERTYPE_Masks
    sortPriority=110
    texture=/BRCosmetics/Characters/Player/Male/Medium/Bodies/M_MED_GlacialTrooper/Textures/T_IceSurface_R
   }
  }
  MaterialExpressionTextureObjectParameter_6 {
   class=MaterialExpressionTextureObjectParameter
   inputs [
    UVs {from="", output=""}
    Apply View MipBias {from="", output=""}
   ]
   properties {
    group=Ice | 02_Surface
    parameterName=Surface - Detail Normal Map
    samplerType=SAMPLERTYPE_Normal
    sortPriority=100
    texture=/BRCosmetics/Characters/Player/Male/Medium/Bodies/M_MED_GlacialTrooper/Textures/T_IceSurface_N
   }
  }
  MaterialExpressionTextureObjectParameter_7 {
   class=MaterialExpressionTextureObjectParameter
   inputs [
    UVs {from="", output=""}
    Apply View MipBias {from="", output=""}
   ]
   properties {group=Ice | 01_Globals, parameterName=Emissive, sortPriority=210, texture=&T_BlacksRGB}
  }
  MaterialExpressionTextureObject_3 {
   class=MaterialExpressionTextureObject
   inputs []
   properties {desc=cube, samplerType=SAMPLERTYPE_LinearColor, texture=/Game/Global/MaterialFunctions/TechArt/Refraction/Shared/T_Latlong_Sky360_256_FN}
  }
  MaterialExpressionTextureObject_4 {
   class=MaterialExpressionTextureObject
   inputs []
   properties {texture=/BRCosmetics/Animation/Props/Gadgets_SatinCheddar/Materials/T_Perlin_Noise_M1}
  }
  MaterialExpressionTextureSampleParameter2D_2 {
   class=MaterialExpressionTextureSampleParameter2D
   inputs [
    UVs {from=MaterialExpressionNamedRerouteUsage_562, output=""}
    "DDX(UVs)" {from=MaterialExpressionNamedRerouteUsage_566, output=""}
    "DDY(UVs)" {from=MaterialExpressionNamedRerouteUsage_567, output=""}
    Apply View MipBias {from="", output=""}
   ]
   properties {
    group=00_Texture Inputs
    mipValueMode=TMVM_Derivative
    parameterName=Diffuse
    samplerSource=SSM_Wrap_WorldGroupSettings
    sortPriority=0
    texture=/Game/Global/Textures/Default/Blanks/T_MidgreysRGB
   }
  }
  MaterialExpressionTextureSample_13 {
   class=MaterialExpressionTextureSample
   inputs [
    UVs {from=MaterialExpressionNamedRerouteUsage_435}
    Tex {from=MaterialExpressionNamedRerouteUsage_451}
    Apply View MipBias {from=""}
   ]
   properties {samplerSource=SSM_Wrap_WorldGroupSettings}
  }
  MaterialExpressionTextureSample_14 {
   class=MaterialExpressionTextureSample
   inputs [
    UVs {from=MaterialExpressionNamedRerouteUsage_436}
    Tex {from=MaterialExpressionNamedRerouteUsage_452}
    Apply View MipBias {from=""}
   ]
   properties {samplerSource=SSM_Wrap_WorldGroupSettings}
  }
  MaterialExpressionTextureSample_16 {
   class=MaterialExpressionTextureSample
   inputs [
    UVs {from=MaterialExpressionNamedRerouteUsage_554}
    Tex {from=MaterialExpressionNamedRerouteUsage_453}
    "DDX(UVs)" {from=MaterialExpressionNamedRerouteUsage_564}
    "DDY(UVs)" {from=MaterialExpressionNamedRerouteUsage_565}
    Apply View MipBias {from=""}
   ]
   properties {mipValueMode=TMVM_Derivative, samplerSource=SSM_Wrap_WorldGroupSettings}
  }
  MaterialExpressionTime_2 {class=MaterialExpressionTime}
  MaterialExpressionTransform_2 {class=MaterialExpressionTransform, inputs [None {from=MaterialExpressionNamedRerouteUsage_481}]}
  MaterialExpressionVectorParameter_0 {
   class=MaterialExpressionVectorParameter
   properties {
    group=Ice | 02_Surface
    parameterName=Surface - Detail Tiles Scale Overrides
    sortPriority=130
    channelNames {a=Bottom Ring, b=Middle Ring, g=Top Ring, r=Rod}
    defaultValue=<rgba 0.1408 0.178 0.177 0.1763>
   }
  }
  MaterialExpressionVectorParameter_3 {
   class=MaterialExpressionVectorParameter
   properties {
    group=Ice | 01_Globals
    parameterName=TechArt Region Masks - Is Half Value
    sortPriority=350
    channelNames {a=Bottom Ring, b=Middle Ring, g=Top Ring, r=Rod}
    defaultValue=<rgba 0 0 1 1>
   }
  }
  MaterialExpressionVectorParameter_44 {
   class=MaterialExpressionVectorParameter
   properties {group=Ice | 04_Volume, parameterName=Volume - Scattering Coeffs, sortPriority=50, defaultValue=<rgba 0 0 0 1>}
  }
  MaterialExpressionVectorParameter_45 {
   class=MaterialExpressionVectorParameter
   properties {
    group=Ice | 06_Emission
    parameterName=Emission - Emissive Fresnel
    sortPriority=10
    channelNames {a=Fresnel Exp, b=Fresnel Tint B, g=Fresnel Tint G, r=Fresnel Tint R}
    defaultValue=<rgba 0 0.3438 0.55 4>
   }
  }
  MaterialExpressionVectorParameter_46 {
   class=MaterialExpressionVectorParameter
   properties {
    group=Ice | 05 _Absorption
    parameterName=Absorption - Diffuse Color Remap
    sortPriority=10
    channelNames {a=Value, b=Saturation, g=Hue Shift, r=Diffuse Map Exp}
    defaultValue=<rgba 1 0 0.5 0.5>
   }
  }
  MaterialExpressionVectorParameter_47 {
   class=MaterialExpressionVectorParameter
   properties {
    group=Ice | 02_Surface
    parameterName=Surface - Detail Normal Controls
    sortPriority=140
    channelNames {a=N/A, b="Min Intensity (Facing)", g="Max Intensity (Grazing)", r=Fresnel Exp}
    defaultValue=<rgba 5 2 0.6 0>
   }
  }
  MaterialExpressionVectorParameter_50 {
   class=MaterialExpressionVectorParameter
   properties {
    group=Ice | 06_Emission
    parameterName=Emission - Variation Noise
    sortPriority=70
    channelNames {a=Pattern Exp, b=Pattern OldMax, g=Pattern OldMin, r=Pattern Scale}
    defaultValue=<rgba 5 0.1 0.8 2.2>
   }
  }
  MaterialExpressionVectorParameter_51 {
   class=MaterialExpressionVectorParameter
   properties {
    group=Ice | 06_Emission
    parameterName=Emission - Noise Depth Breakup
    sortPriority=80
    channelNames {a=Add Constant Depth, b=Breakup Pattern Mult, g=Breakup Pattern Exp, r=Breakup Pattern Scale}
    defaultValue=<rgba 2 0.7 5 1.7>
   }
  }
  MaterialExpressionVectorParameter_52 {
   class=MaterialExpressionVectorParameter
   properties {
    group=Ice | 02_Surface
    parameterName=Surface - Detail Roughness Controls
    sortPriority=150
    channelNames {a=N/A, b=Roughness Mult, g=Roughness Exp, r=Mix Amount}
    defaultValue=<rgba 0 1 1 0>
   }
  }
  MaterialExpressionVectorParameter_53 {
   class=MaterialExpressionVectorParameter
   properties {
    group=Ice | 02_Surface
    parameterName=Surface - Asset Specular Overrides
    sortPriority=10
    channelNames {a=Roughness Mult, b=Roughness Exp, g=Specular Mult, r=Specular Exp}
    defaultValue=<rgba 1 1 1 1>
   }
  }
  MaterialExpressionVectorParameter_54 {
   class=MaterialExpressionVectorParameter
   properties {group=Ice | 01_Globals, parameterName=TOD - Color Morning, sortPriority=410, defaultValue=<rgba 1 1 1 1>}
  }
  MaterialExpressionVectorParameter_55 {
   class=MaterialExpressionVectorParameter
   properties {group=Ice | 01_Globals, parameterName=TOD - Color Day, sortPriority=420, defaultValue=<rgba 1 1 1 1>}
  }
  MaterialExpressionVectorParameter_56 {
   class=MaterialExpressionVectorParameter
   properties {group=Ice | 01_Globals, parameterName=TOD - Color Evening, sortPriority=430, defaultValue=<rgba 1 1 1 1>}
  }
  MaterialExpressionVectorParameter_57 {
   class=MaterialExpressionVectorParameter
   properties {group=Ice | 01_Globals, parameterName=TOD - Color Night, sortPriority=440, defaultValue=<rgba 1 1 1 1>}
  }
  MaterialExpressionVectorParameter_65 {
   class=MaterialExpressionVectorParameter
   properties {group=Ice | 01_Globals, parameterName=Main UV Island, sortPriority=400, defaultValue=<rgba 0.155 0 0.449 0.6>}
  }
  MaterialExpressionVertexNormalWS_2 {class=MaterialExpressionVertexNormalWS, inputs [], properties {}}
 }}
